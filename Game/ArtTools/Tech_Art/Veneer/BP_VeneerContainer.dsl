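; BP_VeneerContainer manages dynamic materials shared with its VeneerComponents, which pull them from the container at BeginPlay.
; Construct Dynamic Materials recreates MIDs whose parent mismatches the base material, then calls SetMaterialParameters.
; Helpers write scalar, vector and texture parameters, including BuildingTextureData maps across four texture slots.
; TransformToMaterialParams packs the inverse scaled actor transform, location, forward vector and scale ratio into four colors for materials.
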(event EventBeginPlay
  (for _array_element (Variables|Default|GetComponents)
    (Class|VeneerComponent|AssignMaterialsfromContainer _array_element)))

(event Custom|Construct%20Dynamic%20Materials
  (params (TransformOnly :type Boolean))
  (for (_array_element _array_index) (Variables|Default|GetVeneerBaseMaterials)
    (if (not (== (Variables|/Script/Engine.MaterialInstance|GetParent (Utilities|Array|Get(acopy) (Variables|Default|GetDynamicMaterials) _array_index)) (Utilities|Struct|BreakStructVeneerOriginalMaterials _array_element)))
      (bind _returnvalue (Rendering|Material|CreateDynamicMaterialInstance (Utilities|Struct|BreakStructVeneerOriginalMaterials _array_element) "None" "None" :!class /Script/Engine.KismetMaterialLibrary))
      (Utilities|Array|SetArrayElem (Variables|Default|GetDynamicMaterials) _array_index _returnvalue true)))
  (CallFunction|SetMaterialParameters false))

(fn ApplyTextureData
  (params (MIC :type Material%20Instance%20Dynamic%20Object%20Reference) (Texture%20Data :type Array%20of%20Building%20Texture%20Data%20Object%20References))
  (LocalVariables|SetLocal%20MIC MIC)
  (for (_array_element _array_index) Texture%20Data
    (LocalVariables|SetLocal%20Index _array_index)
    (Utilities|IsValid (Variables|/Script/FortniteGame.BuildingTextureData|GetDiffuse _array_element)
      (:Is%20Valid
        (Rendering|Material|SetTextureParameterValue (LocalVariables|GetLocal%20MIC) (Utilities|Array|Get(aref) (Utilities|Array|MakeArray "Diffuse" "Diffuse_Texture_2" "Diffuse_Texture_3" "Diffuse_Texture_4") (LocalVariables|GetLocal%20Index)) (Variables|/Script/FortniteGame.BuildingTextureData|GetDiffuse _array_element)))
      (:Is%20Not%20Valid
        _))
    (Utilities|IsValid (Variables|/Script/FortniteGame.BuildingTextureData|GetNormal _array_element)
      (:Is%20Valid
        (Rendering|Material|SetTextureParameterValue (LocalVariables|GetLocal%20MIC) (Utilities|Array|Get(aref) (Utilities|Array|MakeArray "Normals" "Normals_Texture_2" "Normals_Texture_3" "Normals_Texture_4") (LocalVariables|GetLocal%20Index)) (Variables|/Script/FortniteGame.BuildingTextureData|GetNormal _array_element)))
      (:Is%20Not%20Valid
        _))
    (Utilities|IsValid (Variables|/Script/FortniteGame.BuildingTextureData|GetSpecular _array_element)
      (:Is%20Valid
        (Rendering|Material|SetTextureParameterValue (LocalVariables|GetLocal%20MIC) (Utilities|Array|Get(aref) (Utilities|Array|MakeArray "SpecularMasks" "SpecularMasks_2" "SpecularMasks_3" "SpecularMasks_4") (LocalVariables|GetLocal%20Index)) (Variables|/Script/FortniteGame.BuildingTextureData|GetSpecular _array_element))))))

(fn ApplyTextureData
  (params (InputPin :type Material%20Instance%20Dynamic%20Object%20Reference))
  (if true
    _))

(fn SetMaterialScalarParameters
  (params (InputPin :type Material%20Instance%20Dynamic%20Object%20Reference) (Scalar%20Values :type Array%20of%20Float%20%28single-precision%29s) (Param%20Names :type Array%20of%20Names))
  (for (_array_element _array_index) Scalar%20Values
    (Rendering|Material|SetScalarParameterValue InputPin (Utilities|Array|Get(aref) Param%20Names _array_index) _array_element :!class /Script/Engine.MaterialInstanceDynamic)))

(fn SetMaterialTextures
  (params (InputPin :type Material%20Instance%20Dynamic%20Object%20Reference) (Textures :type Array%20of%20Texture%202D%20Object%20References) (Param%20Names :type Array%20of%20Names))
  (for (_array_element _array_index) Textures
    (Rendering|Material|SetTextureParameterValue InputPin (Utilities|Array|Get(aref) Param%20Names _array_index) _array_element)))

(fn SetMaterialVectorParameters
  (params (InputPin :type Material%20Instance%20Dynamic%20Object%20Reference) (Vector%20Values :type Array%20of%20Linear%20Color%20Structures) (Param%20Names :type Array%20of%20Names))
  (for (_array_element _array_index) Vector%20Values
    (Rendering|Material|SetVectorParameterValue InputPin (Utilities|Array|Get(aref) Param%20Names _array_index) _array_element :!class /Script/Engine.MaterialInstanceDynamic)))

(fn TransformToMaterialParams
  (bind (_returnvalue_location _returnvalue_rotation _returnvalue_scale) (Math|Transform|BreakTransform (Transformation|GetActorTransform)))
  (bind (_returnvalue_x _returnvalue_y _returnvalue_z) (Math|Vector|BreakVector (Transformation|GetActorForwardVector)))
  (bind (_returnvalue_x_2 _returnvalue_y_2 _returnvalue_z_2) (Math|Vector|BreakVector (Transformation|GetActorScale3D)))
  (bind (_returnvalue_xplane _returnvalue_yplane _returnvalue_zplane) (Utilities|Struct|BreakMatrix (Math|Matrix|GetInverse(Matrix) (Math|Transform|ToMatrix(Transform) (Math|Transform|MakeTransform _returnvalue_location _returnvalue_rotation (* _returnvalue_scale 1e+02))))))
  (bind (_returnvalue_xplane_w _returnvalue_xplane_x _returnvalue_xplane_y _returnvalue_xplane_z) (Utilities|Struct|BreakPlane _returnvalue_xplane))
  (bind (_returnvalue_yplane_w _returnvalue_yplane_x _returnvalue_yplane_y _returnvalue_yplane_z) (Utilities|Struct|BreakPlane _returnvalue_yplane))
  (bind (_returnvalue_zplane_w _returnvalue_zplane_x _returnvalue_zplane_y _returnvalue_zplane_z) (Utilities|Struct|BreakPlane _returnvalue_zplane))
  (bind (_returnvalue_x_3 _returnvalue_y_3 _returnvalue_z_3) (Math|Vector|BreakVector (Transformation|GetActorLocation)))
  (return (Math|Color|MakeColor _returnvalue_xplane_x _returnvalue_xplane_y _returnvalue_xplane_z _returnvalue_x_3) (Math|Color|MakeColor _returnvalue_yplane_x _returnvalue_yplane_y _returnvalue_yplane_z _returnvalue_y_3) (Math|Color|MakeColor _returnvalue_zplane_x _returnvalue_zplane_y _returnvalue_zplane_z _returnvalue_z_3) (Math|Color|MakeColor _returnvalue_x _returnvalue_y _returnvalue_z (/ _returnvalue_z_2 _returnvalue_y_2))))

(fn UpdateScalarValue
  (params (Param%20Name :type Name) (New%20Value :type Float%20%28double-precision%29))
  (Utilities|Map|Add (Variables|MaterialParameters|GetScalarParams) Param%20Name New%20Value)
  (for _array_element (Variables|Default|GetDynamicMaterials)
    (Rendering|Material|SetScalarParameterValue _array_element Param%20Name New%20Value :!class /Script/Engine.MaterialInstanceDynamic)))

(fn ConstructionScript)
