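; Veneer actor component that applies shared dynamic materials from a BP_VeneerContainer to its owner's static mesh components.
; Gathering lists non-editor-only meshes as veneer structs, optionally resetting the per-component enable array.
; Assignment sets each enabled entry's material slots through pointer indices into the container's dynamic materials, skipping -1.
; UpdateStaticMeshList refreshes component references by array index, keeping flags and pointers; BeginPlay, Bind MIDs Created and Grab Mat are empty.
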(event EventBeginPlay)

(event Custom|Bind%20MIDs%20Created)

(event Custom|Grab%20Mat)

(event Custom|Gather%20Primitive%20Components
  (params (Initilize :type Boolean))
  (if Initilize
    (Utilities|Array|Clear (Variables|Default|GetEnable%20Veener%20On%20Component))
    (else
      _))
  (Utilities|Array|Clear (Variables|Veneer|GetPrimitiveComponentsToVeneer))
  (for _array_element (Actor|GetComponentsByClass (Components|GetOwner) "/Script/Engine.StaticMeshComponent")
    (CallFunction|AddStaticMeshComponenttoList _array_element Initilize)))

(event Custom|Assign%20Materials%20From%20Container
  (for _array_element (Variables|Veneer|GetPrimitiveComponentsToVeneer)
    (if (.bVeneer_8_D78911D64D225FE1D8AB5580B6F6329A (Utilities|Struct|BreakStructVeneer _array_element))
      (for (_array_element_2 _array_index) (.MaterialPointers_6_7ACC7E034D9558D6083707BDAE69BF1D (Utilities|Struct|BreakStructVeneer _array_element))
        (if (!= _array_element_2 -1)
          (Rendering|Material|SetMaterial (Utilities|Struct|BreakStructVeneer _array_element) _array_index (Utilities|Array|Get(acopy) (Variables|/Game/ArtTools/Tech_Art/Veneer/BP_VeneerContainer.BP_VeneerContainer_C|GetDynamicMaterials (Variables|Veneer|GetVeneerContainer)) _array_element_2)))))))

(fn AddStaticMeshComponentToList
  (params (PrimitiveStruct :type Primitive%20Component%20Object%20Reference) (Initilize :type Boolean))
  (if (not (Variables|/Script/Engine.ActorComponent|GetbIsEditorOnly PrimitiveStruct))
    (Utilities|Array|Add (Variables|Veneer|GetPrimitiveComponentsToVeneer) (Utilities|Struct|MakeStructVeneer PrimitiveStruct true))
    (if Initilize
      (Utilities|Array|Add (Variables|Default|GetEnable%20Veener%20On%20Component) "true"))))

(fn UpdateStaticMeshList
  (bind _primitivecomponentstoveneer (Variables|Veneer|GetPrimitiveComponentsToVeneer))
  (for (_array_element _array_index) _primitivecomponentstoveneer
    (Utilities|Array|SetArrayElem _primitivecomponentstoveneer _array_index (Utilities|Struct|MakeStructVeneer (Utilities|Array|Get(acopy) (Actor|GetComponentsByClass (Components|GetOwner) "/Script/Engine.StaticMeshComponent") _array_index) (.bVeneer_8_D78911D64D225FE1D8AB5580B6F6329A (Utilities|Struct|BreakStructVeneer _array_element)) (.MaterialPointers_6_7ACC7E034D9558D6083707BDAE69BF1D (Utilities|Struct|BreakStructVeneer _array_element))))))
