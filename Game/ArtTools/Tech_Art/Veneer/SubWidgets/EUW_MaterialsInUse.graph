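# Editor utility row widget for the Veneer tool that shows a material in use and lets artists mark it for add in source control.
# After a short delay on construct, it asynchronously queries the material's file state.
# Mark For Add saves the asset, marks it and refreshes the parent Veneer lists.
# On failure, Python scans a hardcoded local editor log path for "also opened by" to build the stored error message.
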EditorUtilityWidgetBlueprint /Game/ArtTools/Tech_Art/Veneer/SubWidgets/EUW_MaterialsInUse
graphs [
 EventGraph {nodes [
   n1 = Delay!(execute, WorldContextObject:Object Reference, Duration=2) -> (then)
   AsyncQueryFileState!(
    execute=[@n1.then, @n11.then]
    FileStateCallback:Delegate=Create Event()
    InFile:String=GetSystemPath!(Object:Object Reference=Get Material() -> (Material:Material Interface Object Reference)) -> (ReturnValue:String)
    bSilent:bool=true
   ) -> (then)
   "On Clicked (MarkForAdd)"() {is=title K2Node_ComponentBoundEvent}
   > Set Error(Error:bool=false) -> (Output_Get:bool)
   > n2 = Branch(Condition=IsAvailable!() -> (ReturnValue:bool))
   n3 = Get Material()
   > SaveAsset!(execute, AssetToSave:String=Get Object Path String(Object=@n3), bOnlyIfIsDirty:bool=true) -> (then, ReturnValue:bool)
   > n4 = MarkFileForAdd!(execute, InFile:String=GetSystemPath(Object=@n3), bSilent:bool=false) -> (then, ReturnValue:bool)
   > n5 = Set Error(Error=@n4)
   > n6 = Branch(Condition=@n5)
   > Refresh Materials Lists(self:EUW Veneer Object Reference=Get VeneerWidget() -> (VeneerWidget:EUW Veneer Object Reference))
   n7 = Execute Python Script!(
    execute=@n6.else
    PythonScript="file_path = \"D:\\FN\\M\\FortniteGame\\Saved\\Logs\\FortniteGame.log\"\r\nphrase = \"also opened by\"\r\n\r\nwith open(file_path, 'r') as file:\r\n    lines = file.readlines()\r\n    reversed_lines = reversed(lines)\r\n    for line in reversed_lines:\r\n        if phrase in line:\r\n            found_line = line.strip()\r\n            error = found_line\r\n            break\r\n    else:\r\n        error = \"Phrase not found in the file.\"\r\n        print(\"Phrase not found in the file.\")"
   ) -> (then, ReturnValue, error:String) {outputs [error]}
   n8 = Split(SourceString=Split(SourceString=@n7.error, InStr="also opened by ", SearchDir=FromStart).RightS, InStr="@", SearchDir=FromStart)
   > Set Last Error Msg(Last Error Msg:Text="To Text (String)"(InString=Append!(A="Added but not submitted by: ", B=@n8, C:String="\r\n", D:String=@n8.RightS) -> (ReturnValue))) -> (Output_Get:Text)
   Set Error(execute=@n2.else, Error=true)
   > Set Last Error Msg(Last Error Msg="To Text (String)"(InString=Source Control Not Connected))
   n9 = Query State!() -> (OutputDelegate, then, FileStateOut:Source Control State Structure) {is=title K2Node_CustomEvent}
   n10 = Break Source Control State!(SourceControlState:"Source Control State Structure (by ref)"=@n9.FileStateOut) -> (
    Filename:String
    bIsValid:bool=~
    bIsUnknown:bool=~
    bCanCheckIn:bool=~
    bCanCheckOut:bool=~
    bIsCheckedOut:bool=~
    bIsCurrent:bool=~
    bIsSourceControlled:bool=~
    bIsAdded:bool=~
    bIsDeleted:bool=~
    bIsIgnored:bool=~
    bCanEdit:bool=~
    bCanDelete:bool=~
    bIsModified:bool=~
    bCanAdd:bool=~
    bIsConflicted:bool=~
    bCanRevert:bool=~
    bIsCheckedOutOther:bool=~
    CheckedOutOther:String
    bIsCheckedOutInOtherBranch:bool=~
    bIsModifiedInOtherBranch:bool=~
    PreviousUser:String
   ) {is=title K2Node_BreakStruct}
   > Set Is Added(Is Added:bool=@n10.bIsAdded) -> (Output_Get:bool)
   > Set Checked Out Other(Checked Out Other:String=@n10.CheckedOutOther) -> (Output_Get:String)
   > Set Is Checked Out Other(Is Checked Out Other:bool=@n10.bIsCheckedOutOther) -> (Output_Get:bool)
   Event PreConstruct()
   > SetObject(self=Get SinglePropertyView_24() -> (SinglePropertyView_24:Single Property View Object Reference), NewObject=Self-Reference())
   n11 = Event Construct()
   Event Tick()
  ]}
 Get Text {nodes [
   Get Text() {is=title K2Node_FunctionEntry}
   > Return Node!(execute, ReturnValue:Text="To Text (Boolean)"!(InBool:bool=Get Is Added() -> (Is Added:bool)) -> (ReturnValue:Text))
   "To Text (String)"(InString=Get Filename() -> (Filename:String))
  ]}
 GetBackground {nodes [
   GetBackground() {is=title K2Node_FunctionEntry}
   n1 = Make Slate Color!(SpecifiedColor=(R=0.0075,G=0.0075,B=0.0075,A=1)) -> (SlateColor)
   n2 = Make Slate Color(SpecifiedColor=(R=0.0103,G=0.0103,B=0.0103,A=1))
   n3 = Get IsDarkBG() -> (IsDarkBG:bool)
   > Return Node(ReturnValue:Slate Brush Structure=Make Slate Brush!(TintColor:Slate Color Structure=Select(Option 0:Slate Color Structure=@n2, Option 1:Slate Color Structure=@n1, Index=@n3) -> (ReturnValue:Slate Color Structure)) -> (SlateBrush:Slate Brush Structure) {is=title K2Node_MakeStruct})
  ]}
 GetText {nodes [
   GetText() {is=title K2Node_FunctionEntry}
   > Return Node(ReturnValue:Text=Get Last Error Msg() -> (Last Error Msg:Text))
   "To Text (String)"(InString=Get Checked Out Other() -> (Checked Out Other:String))
  ]}
 GetVisibility_0 {nodes [
   GetVisibility_0() {is=title K2Node_FunctionEntry}
   > Return Node(ReturnValue:ESlateVisibility Enum=Select(Option 0:ESlateVisibility Enum=Visible, Option 1:ESlateVisibility Enum=Collapsed, Index=Get IsSupported() -> (IsSupported:bool)) -> (ReturnValue:ESlateVisibility Enum))
  ]}
 GetVisibility_1 {nodes [
   GetVisibility_1() {is=title K2Node_FunctionEntry}
   > Return Node(ReturnValue=Select(Option 0=Collapsed, Option 1=Visible, Index=Get IsSupported()))
  ]}
 GetVisibility_2 {nodes [
   GetVisibility_2() {is=title K2Node_FunctionEntry}
   > Return Node(ReturnValue=Visible)
   n1 = Get Error() -> (Error:bool)
   AND Boolean(A=NOT Boolean(A=Get Is Added()), B=@n1)
   Select(Option 0=Collapsed, Option 1=Visible, Index=NOT Boolean(A=@n1))
  ]}
 Get_MarkForAdd_Visibility {nodes [
   Get_MarkForAdd_Visibility() {is=title K2Node_FunctionEntry}
   > Return Node(ReturnValue=Select(Option 0=Visible, Option 1=Collapsed, Index=Get Is Added()))
  ]}
]
variables [
 Checked Out Other {type=string, default=""}
 Error {type=boolean}
 Filename {type=string, default=""}
 Is Added {type=boolean}
 Is Checked Out Other {type=boolean}
 IsDarkBG {type=boolean}
 IsSupported {type=boolean}
 Last Error Msg {type=string, default=""}
 Material {type=MaterialInterface, default=null}
 VeneerWidget {type=EUW_Veneer_C, default=null}
]
widgets {
 info {rootWidgetClass=&CanvasPanel, widgetCount=8}
 widgets [
  CanvasPanel_0 {widgetClassPath=&CanvasPanel, widget=CanvasPanel_0}
  Border_37 {
   widgetClassPath=&Border
   parent=CanvasPanel_0
   widget=Border_37
   properties {background {
     bIsDynamicallyLoaded=false
     drawAs=Image
     imageType=NoImage
     mirroring=NoMirror
     resourceName=None
     resourceObject=null
     tiling=NoTile
     imageSize=<32 32>
     margin=<ltrb 0 0 0 0>
     outlineSettings {
      bUseBrushTransparency=false
      roundingType=HalfHeightRadius
      width=0
      color {colorUseRule=UseColor_Specified, specifiedColor=<rgba 0 0 0 0>}
      cornerRadii=<0 0 0 0>
     }
     tintColor {colorUseRule=UseColor_Specified, specifiedColor=<rgba 0.1195 0.1195 0.1195 1>}
     uVRegion {bIsValid=false, max=<0 0>, min=<0 0>}
    }}
   slot {class=CanvasPanelSlot, properties {layoutData {alignment=<0 0>, anchors {maximum=<1 0>, minimum=<0 0>}, offsets=<ltrb 0 0 0 80>}}}
  }
  HorizontalBox_0 {widgetClassPath=&HorizontalBox, parent=Border_37, widget=HorizontalBox_0, slot {class=BorderSlot}}
  SinglePropertyView_24 {
   widgetClassPath=&SinglePropertyView
   parent=HorizontalBox_0
   widget=SinglePropertyView_24
   properties {nameOverride=" ", propertyName=Material}
   slot {class=HorizontalBoxSlot, properties {size {sizeRule=Fill, value=1}}}
  }
  VerticalBox_27 {
   widgetClassPath=&VerticalBox
   parent=HorizontalBox_0
   widget=VerticalBox_27
   slot {class=HorizontalBoxSlot, properties {horizontalAlignment=HAlign_Center, verticalAlignment=VAlign_Center, size {sizeRule=Fill, value=1}}}
  }
  MarkForAdd {
   widgetClassPath=&EditorUtilityButton
   parent=VerticalBox_27
   widget=MarkForAdd
   slot {class=VerticalBoxSlot, properties {horizontalAlignment=HAlign_Center, verticalAlignment=VAlign_Center}}
  }
  TextBlock_59 {
   widgetClassPath=&TextBlock
   parent=MarkForAdd
   widget=TextBlock_59
   properties {
    text=Mark For Add
    font {
     bForceMonospaced=false
     bMaterialIsStencil=false
     fontMaterial=null
     fontObject=&Roboto
     letterSpacing=0
     monospacedWidth=1
     size=8.25
     skewAmount=0
     typefaceFontName=Bold
     outlineSettings {
      bApplyOutlineToDropShadows=false
      bMiteredCorners=false
      bSeparateFillAlpha=false
      outlineMaterial=null
      outlineSize=0
      outlineColor=<rgba 0 0 0 1>
     }
    }
   }
   slot {class=ButtonSlot, properties {padding=<ltrb 0 0 0 0>}}
  }
  TextBlock_83 {
   widgetClassPath=&TextBlock
   parent=VerticalBox_27
   widget=TextBlock_83
   properties {
    text=Text Block
    colorAndOpacity {colorUseRule=UseColor_Specified, specifiedColor=<rgba 1 0.7157 0 1>}
    font {
     bForceMonospaced=false
     bMaterialIsStencil=false
     fontMaterial=null
     fontObject=&Roboto
     letterSpacing=0
     monospacedWidth=1
     size=8.25
     skewAmount=0
     typefaceFontName=Bold
     outlineSettings {
      bApplyOutlineToDropShadows=false
      bMiteredCorners=false
      bSeparateFillAlpha=false
      outlineMaterial=null
      outlineSize=0
      outlineColor=<rgba 0 0 0 1>
     }
    }
   }
   slot {class=VerticalBoxSlot, properties {horizontalAlignment=HAlign_Center}}
  }
 ]
}
