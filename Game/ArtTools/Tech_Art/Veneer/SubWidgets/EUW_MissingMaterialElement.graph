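# Editor utility sub-widget for the Veneer art tool: one row for a material missing its Veneer material instance constant.
# A property view bound to itself shows the Material, and IsSupported-driven getters select between an unsupported warning and creation controls.
# The button calls CreateMIC on the parent VeneerWidget, then refreshes its lists.
# A macro builds the displayed target path from a VNR subfolder and a veneer-type suffix.
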EditorUtilityWidgetBlueprint /Game/ArtTools/Tech_Art/Veneer/SubWidgets/EUW_MissingMaterialElement
graphs [
 ConvertOriginalPathToVRN {nodes [
   n1 = Split(SourceString=Inputs!() -> (SourceString:String), InStr=/, SearchDir=FromEnd)
   n2 = "To Integer (Byte)"!(InByte:Byte=Get VeneerTypeBuffer!(self:EUW Veneer Object Reference=Get VeneerWidget() -> (VeneerWidget:EUW Veneer Object Reference)) -> (VeneerTypeBuffer:Enum_VeneerType Enum=NewEnumerator0)) -> (ReturnValue:int=0)
   n3 = Make Array!("[0]":String=D, "[1]":String=DA, "[2]":String=C, "[3]":String=L) -> (Array:Array of Strings)
   Outputs!(
    Asset Name:String=Append!(
     A=Split(SourceString=@n1.RightS, InStr=., SearchDir=FromEnd)
     B=_VNR_
     C:String="Get (a ref)"(Array:Array of Strings=@n3, Dimension 1=@n2) -> (Output:"String (by ref)")
    ) -> (ReturnValue)
    Asset Path:String=Append!(A=@n1, B=/VNR/) -> (ReturnValue)
   )
  ]}
 EventGraph {nodes [
   "On Clicked (EditorUtilityButton_42)"() {is=title K2Node_ComponentBoundEvent}
   > CreateMIC!(
    execute
    self:EUW Veneer Object Reference=Get VeneerWidget()
    Material:Material Interface Object Reference=Get Material() -> (Material:Material Interface Object Reference)
   ) -> (then)
   > Refresh Materials Lists(self:EUW Veneer Object Reference=Get VeneerWidget())
   Remove Create MIC() {is=title K2Node_CustomEvent}
   > CreateMIC(self=Get VeneerWidget(), Material=Get Material())
   Event PreConstruct()
   > SetObject(self=Get SinglePropertyView_24() -> (SinglePropertyView_24:Single Property View Object Reference), NewObject=Self-Reference())
   Event Construct()
   Event Tick()
  ]}
 GetBackground {nodes [
   GetBackground() {is=title K2Node_FunctionEntry}
   n1 = Make Slate Color!(SpecifiedColor=(R=0.0075,G=0.0075,B=0.0075,A=1)) -> (SlateColor)
   n2 = Make Slate Color(SpecifiedColor=(R=0.0103,G=0.0103,B=0.0103,A=1))
   n3 = Get IsDarkBG() -> (IsDarkBG:bool)
   > Return Node!(
    execute
    ReturnValue:Slate Brush Structure=Make Slate Brush!(TintColor:Slate Color Structure=Select(Option 0:Slate Color Structure=@n2, Option 1:Slate Color Structure=@n1, Index=@n3) -> (ReturnValue:Slate Color Structure)) -> (SlateBrush:Slate Brush Structure) {is=title K2Node_MakeStruct}
   )
  ]}
 GetText {nodes [
   GetText() {is=title K2Node_FunctionEntry}
   > n1 = GetPathNameForLoadedAsset(LoadedAsset=Get Material())
   n2 = Convert Original Path to VRN!(SourceString:String=@n1) -> (Asset Name:String, Asset Path:String) {is=title K2Node_MacroInstance}
   > Return Node(ReturnValue:Text="To Text (String)"(InString=Append(A=@n2.Asset Path, B=@n2)))
  ]}
 GetVisibility_0 {nodes [
   GetVisibility_0() {is=title K2Node_FunctionEntry}
   > Return Node(ReturnValue:ESlateVisibility Enum=Select(Option 0:ESlateVisibility Enum=Visible, Option 1:ESlateVisibility Enum=Collapsed, Index=Get IsSupported() -> (IsSupported:bool)) -> (ReturnValue:ESlateVisibility Enum))
  ]}
 GetVisibility_1 {nodes [
   GetVisibility_1() {is=title K2Node_FunctionEntry}
   > Return Node(ReturnValue=Select(Option 0=Collapsed, Option 1=Visible, Index=Get IsSupported()))
  ]}
]
variables [
 IsDarkBG {type=boolean}
 IsSupported {type=boolean}
 Material {type=MaterialInterface, default=null}
 VeneerWidget {type=EUW_Veneer_C, default=null}
]
widgets {
 info {rootWidgetClass=&CanvasPanel, widgetCount=11}
 widgets [
  CanvasPanel_0 {widgetClassPath=&CanvasPanel, widget=CanvasPanel_0}
  Border_37 {
   widgetClassPath=&Border
   parent=CanvasPanel_0
   widget=Border_37
   properties {background {
     bIsDynamicallyLoaded=false
     drawAs=Image
     imageType=NoImage
     mirroring=NoMirror
     resourceName=None
     resourceObject=null
     tiling=NoTile
     imageSize=<32 32>
     margin=<ltrb 0 0 0 0>
     outlineSettings {
      bUseBrushTransparency=false
      roundingType=HalfHeightRadius
      width=0
      color {colorUseRule=UseColor_Specified, specifiedColor=<rgba 0 0 0 0>}
      cornerRadii=<0 0 0 0>
     }
     tintColor {colorUseRule=UseColor_Specified, specifiedColor=<rgba 0.1195 0.1195 0.1195 1>}
     uVRegion {bIsValid=false, max=<0 0>, min=<0 0>}
    }}
   slot {class=CanvasPanelSlot, properties {layoutData {alignment=<0 0>, anchors {maximum=<1 0>, minimum=<0 0>}, offsets=<ltrb 0 0 55.8571 80>}}}
  }
  SinglePropertyView_24 {
   widgetClassPath=&SinglePropertyView
   parent=CanvasPanel_0
   widget=SinglePropertyView_24
   properties {nameOverride=" ", propertyName=Material}
   slot {class=CanvasPanelSlot, properties {layoutData {alignment=<0 0>, anchors {maximum=<1 0>, minimum=<0 0>}, offsets=<ltrb 0 0 0 80>}}}
  }
  Border_174 {
   widgetClassPath=&Border
   parent=CanvasPanel_0
   widget=Border_174
   properties {background {
     bIsDynamicallyLoaded=false
     drawAs=Image
     imageType=NoImage
     mirroring=NoMirror
     resourceName=None
     resourceObject=null
     tiling=NoTile
     imageSize=<32 32>
     margin=<ltrb 0 0 0 0>
     outlineSettings {
      bUseBrushTransparency=false
      roundingType=HalfHeightRadius
      width=0
      color {colorUseRule=UseColor_Specified, specifiedColor=<rgba 0 0 0 0>}
      cornerRadii=<0 0 0 0>
     }
     tintColor {colorUseRule=UseColor_Specified, specifiedColor=<rgba 0.1198 0.1198 0.1198 1>}
     uVRegion {bIsValid=false, max=<0 0>, min=<0 0>}
    }}
   slot {
    class=CanvasPanelSlot
    properties {zOrder=1, layoutData {alignment=<0 0>, anchors {maximum=<1 0>, minimum=<0 0>}, offsets=<ltrb 73.205 0 0 80>}}
   }
  }
  HorizontalBox_36 {widgetClassPath=&HorizontalBox, parent=Border_174, widget=HorizontalBox_36, slot {class=BorderSlot}}
  HorizontalBox_62 {
   widgetClassPath=&HorizontalBox
   parent=HorizontalBox_36
   widget=HorizontalBox_62
   properties {visibility=Visible}
   slot {class=HorizontalBoxSlot, properties {size {sizeRule=Fill, value=1}}}
  }
  TextBlock_35 {
   widgetClassPath=&TextBlock
   parent=HorizontalBox_62
   widget=TextBlock_35
   properties {
    text="This material doesn't support\r\nVeneer Functions"
    colorAndOpacity {colorUseRule=UseColor_Specified, specifiedColor=<rgba 1 0.7011 0 1>}
    font {
     bForceMonospaced=false
     bMaterialIsStencil=false
     fontMaterial=null
     fontObject=&Roboto
     letterSpacing=0
     monospacedWidth=1
     size=9
     skewAmount=0
     typefaceFontName=Bold
     outlineSettings {
      bApplyOutlineToDropShadows=false
      bMiteredCorners=false
      bSeparateFillAlpha=false
      outlineMaterial=null
      outlineSize=0
      outlineColor=<rgba 0 0 0 1>
     }
    }
   }
   slot {class=HorizontalBoxSlot, properties {horizontalAlignment=HAlign_Center, verticalAlignment=VAlign_Center, size {sizeRule=Fill, value=1}}}
  }
  VerticalBox_0 {
   widgetClassPath=&VerticalBox
   parent=HorizontalBox_36
   widget=VerticalBox_0
   slot {class=HorizontalBoxSlot, properties {size {sizeRule=Fill, value=1}}}
  }
  EditorUtilityButton_42 {
   widgetClassPath=&EditorUtilityButton
   parent=VerticalBox_0
   widget=EditorUtilityButton_42
   slot {
    class=VerticalBoxSlot
    properties {horizontalAlignment=HAlign_Center, verticalAlignment=VAlign_Top, padding=<ltrb 0 8 0 0>, size {sizeRule=Fill, value=1}}
   }
  }
  TextBlock_86 {
   widgetClassPath=&TextBlock
   parent=EditorUtilityButton_42
   widget=TextBlock_86
   properties {
    text=Create Material Constant
    font {
     bForceMonospaced=false
     bMaterialIsStencil=false
     fontMaterial=null
     fontObject=&Roboto
     letterSpacing=0
     monospacedWidth=1
     size=8.25
     skewAmount=0
     typefaceFontName=Regular
     outlineSettings {
      bApplyOutlineToDropShadows=false
      bMiteredCorners=false
      bSeparateFillAlpha=false
      outlineMaterial=null
      outlineSize=0
      outlineColor=<rgba 0 0 0 1>
     }
    }
   }
   slot {class=ButtonSlot}
  }
  TextBlock_61 {
   widgetClassPath=&TextBlock
   parent=VerticalBox_0
   widget=TextBlock_61
   properties {
    autoWrapText=true
    text="Path: Something Something Dark Side, Something Something Dark Force"
    font {
     bForceMonospaced=false
     bMaterialIsStencil=false
     fontMaterial=null
     fontObject=&Roboto
     letterSpacing=0
     monospacedWidth=1
     size=8.25
     skewAmount=0
     typefaceFontName=Bold
     outlineSettings {
      bApplyOutlineToDropShadows=false
      bMiteredCorners=false
      bSeparateFillAlpha=false
      outlineMaterial=null
      outlineSize=0
      outlineColor=<rgba 0 0 0 1>
     }
    }
   }
   slot {class=VerticalBoxSlot, properties {verticalAlignment=VAlign_Top, size {sizeRule=Fill, value=1}}}
  }
 ]
}
