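# Editor utility sub-widget for Veneer: a labeled spin box editing one named scalar material parameter on a BP_VeneerContainer.
# Construct applies slider limits, then loads the value from the container's ScalarParams map or adds the default.
# Slider drags (gated, with Slate throttling disabled) and commits write to the map and push the value to the container's dynamic materials.
# Strict Min Max hard-clamps only the minimum, re-setting the slider maximum instead of MaxValue.
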EditorUtilityWidgetBlueprint /Game/ArtTools/Tech_Art/Veneer/SubWidgets/EUW_VeneerScalarParam
graphs [
 EventGraph {nodes [
   n1 = "On Value Changed (EditorUtilitySpinBox_21)"!() -> (OutputDelegate, then, InValue:float) {is=title K2Node_ComponentBoundEvent}
   > Gate!(Enter:Exec, Open:Exec=@n2.then, Close:Exec=@n3.then, Toggle:Exec, bStartClosed:bool=true) -> (Exit:Exec) {is=title K2Node_MacroInstance}
   > Add!(
    execute
    TargetMap:"Map of Names to Float (double-precision)s"=Get ScalarParams!(self:BP Veneer Container Object Reference=Get Veneer Container() -> (Veneer Container:BP Veneer Container Object Reference)) -> (ScalarParams:"Map of Names to Float (double-precision)s")
    Key:"Name (by ref)"=Get ParameterName() -> (ParameterName:Name=None)
    Value:"Float (double-precision) (by ref)"=@n1.InValue
   ) -> (then) {is=title K2Node_CallFunction}
   > UpdateContainerMIDs!(execute, self, Value:float=@n1.InValue) -> (then)
   "On Begin Slider Movement (EditorUtilitySpinBox_21)"() {is=title K2Node_ComponentBoundEvent}
   > n2 = ExecuteConsoleCommand(Command=Slate.bAllowThrottling 0)
   "On End Slider Movement (EditorUtilitySpinBox_21)"!() -> (OutputDelegate, then, InValue:float) {is=title K2Node_ComponentBoundEvent}
   > n3 = ExecuteConsoleCommand(Command=Slate.bAllowThrottling 1)
   n4 = "On Value Committed (EditorUtilitySpinBox_21)"!() -> (OutputDelegate, then, InValue:float, CommitMethod:ETextCommit Enum=Default) {is=title K2Node_ComponentBoundEvent}
   > Add(TargetMap=Get ScalarParams(self=Get Veneer Container()), Key=Get ParameterName() -> (ParameterName=None), Value=@n4.InValue)
   > UpdateContainerMIDs(Value=@n4.InValue)
   Event Construct()
   > n5 = Sequence()
   > Set MinSliderValue!(
    execute
    MinSliderValue:float=Get SpinnerMin() -> (SpinnerMin:double)
    self:Spin Box Object Reference=Get EditorUtilitySpinBox_21() -> (EditorUtilitySpinBox_21:Editor Utility Spin Box Object Reference)
   ) -> (then, Output_Get:float)
   > Set MaxSliderValue!(execute, MaxSliderValue:float=Get SpinnerMax() -> (SpinnerMax:double), self:Spin Box Object Reference=Get EditorUtilitySpinBox_21()) -> (then, Output_Get:float)
   > Set Value!(execute, Value:float=Get Default Value() -> (Default Value:double), self:Spin Box Object Reference=Get EditorUtilitySpinBox_21()) -> (then, Output_Get:float)
   > Branch(Condition=Get Strict Min Max() -> (Strict Min Max:bool))
   > Set MinValue!(execute, MinValue:float=Get SpinnerMin(), self:Spin Box Object Reference=Get EditorUtilitySpinBox_21()) -> (then, Output_Get:float)
   > Set MaxSliderValue(MaxSliderValue=Get SpinnerMax(), self=Get EditorUtilitySpinBox_21())
   n6 = Find(
    TargetMap:"Map of Names to Float (double-precision)s"=Get ScalarParams(self=Get Veneer Container())
    Key:"Name (by ref)"=Get ParameterName() -> (ParameterName=None)
   ) -> (Value:double)
   n7 = Branch(execute=@n5.then_1, Condition=@n6.ReturnValue)
   > Set Value(Value=@n6, self=Get EditorUtilitySpinBox_21())
   Add(
    execute=@n7.else
    TargetMap=Get ScalarParams(self=Get Veneer Container())
    Key=Get ParameterName() -> (ParameterName=None)
    Value=Get Default Value()
   )
   Event PreConstruct()
  ]}
 GetText {nodes [
   GetText() {is=title K2Node_FunctionEntry}
   > Return Node!(execute, ReturnValue:Text="To Text (Name)"!(InName:Name=Get FriendlyName() -> (FriendlyName:Name=None)) -> (ReturnValue:Text))
  ]}
 GetToolTipText {nodes [
   GetToolTipText() {is=title K2Node_FunctionEntry}
   > Return Node(ReturnValue="To Text (Name)"(InName=Get ParameterName() -> (ParameterName=None)))
  ]}
 UpdateContainerMIDs {nodes [
   n1 = UpdateContainerMIDs!() -> (then, Value:float=~) {is=title K2Node_FunctionEntry}
   n2 = Get Veneer Container()
   > Is Valid(InputObject=@n2)
   > n3 = For Each Loop(Array:Array of Material Instance Dynamic Object References=Get DynamicMaterials!(self:BP Veneer Container Object Reference=@n2) -> (DynamicMaterials:Array of Material Instance Dynamic Object References)) -> (Array Element:Material Instance Dynamic Object Reference)
   > SetScalarParameterValue(self=@n3, ParameterName=Get ParameterName() -> (ParameterName=None), Value=@n1)
  ]}
]
variables [
 Default Value {type=number}
 FriendlyName {type=string, default=Param Name}
 ParameterName {type=string, default=Param}
 ScalarValue {type=number}
 SpinnerMax {type=number}
 SpinnerMin {type=number}
 Strict Min Max {type=boolean}
 Veneer Container {type=BP_VeneerContainer_C, default=null}
]
widgets {
 info {rootWidgetClass=&CanvasPanel, widgetCount=7}
 widgets [
  CanvasPanel_0 {widgetClassPath=&CanvasPanel, widget=CanvasPanel_0}
  VerticalBox_1 {
   widgetClassPath=&VerticalBox
   parent=CanvasPanel_0
   widget=VerticalBox_1
   slot {class=CanvasPanelSlot, properties {layoutData {alignment=<0 0>, anchors {maximum=<1 0>, minimum=<0 0>}, offsets=<ltrb 0 0 0 26>}}}
  }
  HorizontalBox_0 {widgetClassPath=&HorizontalBox, parent=VerticalBox_1, widget=HorizontalBox_0, slot {class=VerticalBoxSlot}}
  TextBlock_63 {
   widgetClassPath=&TextBlock
   parent=HorizontalBox_0
   widget=TextBlock_63
   properties {
    text=Text Block
    colorAndOpacity {colorUseRule=UseColor_Specified, specifiedColor=<rgba 0.3813 0.3813 0.3813 1>}
    font {
     bForceMonospaced=false
     bMaterialIsStencil=false
     fontMaterial=null
     fontObject=&Roboto
     letterSpacing=0
     monospacedWidth=1
     size=8.25
     skewAmount=0
     typefaceFontName=Regular
     outlineSettings {
      bApplyOutlineToDropShadows=false
      bMiteredCorners=false
      bSeparateFillAlpha=false
      outlineMaterial=null
      outlineSize=0
      outlineColor=<rgba 0 0 0 1>
     }
    }
   }
   slot {
    class=HorizontalBoxSlot
    properties {horizontalAlignment=HAlign_Left, verticalAlignment=VAlign_Center, padding=<ltrb 24 0 0 0>, size {sizeRule=Fill, value=0.5}}
   }
  }
  Border {
   widgetClassPath=&Border
   parent=HorizontalBox_0
   widget=Border
   properties {
    background {
     bIsDynamicallyLoaded=false
     drawAs=Image
     imageType=NoImage
     mirroring=NoMirror
     resourceName=None
     resourceObject=null
     tiling=NoTile
     imageSize=<32 32>
     margin=<ltrb 0 0 0 0>
     outlineSettings {
      bUseBrushTransparency=false
      roundingType=HalfHeightRadius
      width=0
      color {colorUseRule=UseColor_Specified, specifiedColor=<rgba 0 0 0 0>}
      cornerRadii=<0 0 0 0>
     }
     tintColor {colorUseRule=UseColor_Specified, specifiedColor=<rgba 0.0103 0.0103 0.0103 1>}
     uVRegion {bIsValid=false, max=<0 0>, min=<0 0>}
    }
    contentColorAndOpacity=<rgba 0.0103 0.0103 0.0103 1>
    padding=<ltrb 1 0 0 0>
   }
   slot {class=HorizontalBoxSlot, properties {padding=<ltrb 5 0 5 0>}}
  }
  EditorUtilitySpinBox_21 {
   widgetClassPath=/Script/Blutility.EditorUtilitySpinBox
   bIsVariable=true
   parent=HorizontalBox_0
   widget=EditorUtilitySpinBox_21
   properties {
    maxFractionalDigits=3
    minDesiredWidth=100
    font {
     bForceMonospaced=false
     bMaterialIsStencil=false
     fontMaterial=null
     fontObject=&Roboto
     letterSpacing=0
     monospacedWidth=1
     size=8.25
     skewAmount=0
     typefaceFontName=Regular
     outlineSettings {
      bApplyOutlineToDropShadows=false
      bMiteredCorners=false
      bSeparateFillAlpha=false
      outlineMaterial=null
      outlineSize=0
      outlineColor=<rgba 0 0 0 1>
     }
    }
    foregroundColor {colorUseRule=UseColor_Specified, specifiedColor=<rgba 1 1 1 1>}
   }
   slot {class=HorizontalBoxSlot, properties {horizontalAlignment=HAlign_Left, verticalAlignment=VAlign_Center, size {sizeRule=Fill, value=1}}}
  }
  Border_36 {
   widgetClassPath=&Border
   parent=VerticalBox_1
   widget=Border_36
   properties {
    background {
     bIsDynamicallyLoaded=false
     drawAs=Image
     imageType=NoImage
     mirroring=NoMirror
     resourceName=None
     resourceObject=null
     tiling=NoTile
     imageSize=<32 32>
     margin=<ltrb 0 0 0 0>
     outlineSettings {
      bUseBrushTransparency=false
      roundingType=HalfHeightRadius
      width=0
      color {colorUseRule=UseColor_Specified, specifiedColor=<rgba 0 0 0 0>}
      cornerRadii=<0 0 0 0>
     }
     tintColor {colorUseRule=UseColor_Specified, specifiedColor=<rgba 0.0103 0.0103 0.0103 1>}
     uVRegion {bIsValid=false, max=<0 0>, min=<0 0>}
    }
    contentColorAndOpacity=<rgba 0.0103 0.0103 0.0103 1>
    padding=<ltrb 0 0 0 1>
   }
   slot {class=VerticalBoxSlot}
  }
 ]
}
