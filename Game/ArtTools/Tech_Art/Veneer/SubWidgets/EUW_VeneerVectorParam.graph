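# Editor utility widget row for the Veneer art tool that edits one named vector (linear color) parameter on a Veneer Container.
# PreConstruct points a SinglePropertyView at the widget itself, exposing its own Vector Value for editing.
# Construct loads that value from the container's VectorParams map by ParameterName, or adds the default if missing.
# Edits write back to the map, and GetText supplies FriendlyName as display text.
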EditorUtilityWidgetBlueprint /Game/ArtTools/Tech_Art/Veneer/SubWidgets/EUW_VeneerVectorParam
graphs [
 EventGraph {nodes [
   "On Property Changed (SinglePropertyView_108)"!() -> (OutputDelegate, then, PropertyName:Name=None) {is=title K2Node_ComponentBoundEvent}
   > Add!(
    execute
    TargetMap:Map of Names to Linear Color Structures=Get VectorParams!(self:BP Veneer Container Object Reference=Get Veneer Container() -> (Veneer Container:BP Veneer Container Object Reference)) -> (VectorParams:Map of Names to Linear Color Structures)
    Key:"Name (by ref)"=Get ParameterName() -> (ParameterName:Name=None)
    Value:"Linear Color Structure (by ref)"=Get Vector Value() -> (Vector Value:Linear Color Structure)
   ) -> (then) {is=title K2Node_CallFunction}
   Event PreConstruct()
   > SetObject(self=Get SinglePropertyView_108() -> (SinglePropertyView_108:Single Property View Object Reference), NewObject=Self-Reference())
   Event Construct()
   n1 = Get VectorParams(self=Get Veneer Container())
   n2 = Find(TargetMap:Map of Names to Linear Color Structures=@n1, Key:"Name (by ref)"=Get ParameterName() -> (ParameterName=None)) -> (Value:Linear Color Structure)
   > n3 = Branch(Condition=@n2.ReturnValue)
   > Set Vector Value(Vector Value:Linear Color Structure=@n2) -> (Output_Get:Linear Color Structure)
   Add(execute=@n3.else, TargetMap=@n1, Key=Get ParameterName() -> (ParameterName=None), Value=Get Vector Value())
  ]}
 GetText {nodes [
   GetText() {is=title K2Node_FunctionEntry}
   > Return Node!(execute, ReturnValue:Text="To Text (Name)"!(InName:Name=Get FriendlyName() -> (FriendlyName:Name=None)) -> (ReturnValue:Text))
  ]}
]
variables [
 FriendlyName {type=string, default=Param Name}
 ParameterName {type=string, default=Param}
 Vector Value {type=LinearColor, default=<rgba 1 1 1 1>}
 Veneer Container {type=BP_VeneerContainer_C, default=null}
]
widgets {
 info {rootWidgetClass=&CanvasPanel, widgetCount=7}
 widgets [
  CanvasPanel_0 {widgetClassPath=&CanvasPanel, widget=CanvasPanel_0}
  VerticalBox_0 {
   widgetClassPath=&VerticalBox
   parent=CanvasPanel_0
   widget=VerticalBox_0
   slot {class=CanvasPanelSlot, properties {layoutData {alignment=<0 0>, anchors {maximum=<1 0>, minimum=<0 0>}, offsets=<ltrb 0 0 0 26>}}}
  }
  HorizontalBox_0 {widgetClassPath=&HorizontalBox, parent=VerticalBox_0, widget=HorizontalBox_0, slot {class=VerticalBoxSlot}}
  TextBlock_63 {
   widgetClassPath=&TextBlock
   parent=HorizontalBox_0
   widget=TextBlock_63
   properties {
    text=Text Block
    colorAndOpacity {colorUseRule=UseColor_Specified, specifiedColor=<rgba 0.3813 0.3813 0.3813 1>}
    font {
     bForceMonospaced=false
     bMaterialIsStencil=false
     fontMaterial=null
     fontObject=&Roboto
     letterSpacing=0
     monospacedWidth=1
     size=8.25
     skewAmount=0
     typefaceFontName=Regular
     outlineSettings {
      bApplyOutlineToDropShadows=false
      bMiteredCorners=false
      bSeparateFillAlpha=false
      outlineMaterial=null
      outlineSize=0
      outlineColor=<rgba 0 0 0 1>
     }
    }
   }
   slot {
    class=HorizontalBoxSlot
    properties {horizontalAlignment=HAlign_Left, verticalAlignment=VAlign_Center, padding=<ltrb 24 0 0 0>, size {sizeRule=Fill, value=0.5}}
   }
  }
  Border {
   widgetClassPath=&Border
   parent=HorizontalBox_0
   widget=Border
   properties {
    background {
     bIsDynamicallyLoaded=false
     drawAs=Image
     imageType=NoImage
     mirroring=NoMirror
     resourceName=None
     resourceObject=null
     tiling=NoTile
     imageSize=<32 32>
     margin=<ltrb 0 0 0 0>
     outlineSettings {
      bUseBrushTransparency=false
      roundingType=HalfHeightRadius
      width=0
      color {colorUseRule=UseColor_Specified, specifiedColor=<rgba 0 0 0 0>}
      cornerRadii=<0 0 0 0>
     }
     tintColor {colorUseRule=UseColor_Specified, specifiedColor=<rgba 0.0103 0.0103 0.0103 1>}
     uVRegion {bIsValid=false, max=<0 0>, min=<0 0>}
    }
    contentColorAndOpacity=<rgba 0.0103 0.0103 0.0103 1>
    padding=<ltrb 1 0 0 0>
   }
   slot {class=HorizontalBoxSlot, properties {padding=<ltrb 5 0 5 0>}}
  }
  SinglePropertyView_108 {
   widgetClassPath=&SinglePropertyView
   parent=HorizontalBox_0
   widget=SinglePropertyView_108
   properties {nameOverride=" ", propertyName=Vector Value}
   slot {class=HorizontalBoxSlot, properties {horizontalAlignment=HAlign_Left, verticalAlignment=VAlign_Center, size {sizeRule=Fill, value=1}}}
  }
  Border_36 {
   widgetClassPath=&Border
   parent=VerticalBox_0
   widget=Border_36
   properties {
    background {
     bIsDynamicallyLoaded=false
     drawAs=Image
     imageType=NoImage
     mirroring=NoMirror
     resourceName=None
     resourceObject=null
     tiling=NoTile
     imageSize=<32 32>
     margin=<ltrb 0 0 0 0>
     outlineSettings {
      bUseBrushTransparency=false
      roundingType=HalfHeightRadius
      width=0
      color {colorUseRule=UseColor_Specified, specifiedColor=<rgba 0 0 0 0>}
      cornerRadii=<0 0 0 0>
     }
     tintColor {colorUseRule=UseColor_Specified, specifiedColor=<rgba 0.0103 0.0103 0.0103 1>}
     uVRegion {bIsValid=false, max=<0 0>, min=<0 0>}
    }
    contentColorAndOpacity=<rgba 0.0103 0.0103 0.0103 1>
    padding=<ltrb 0 0 0 1>
   }
   slot {class=VerticalBoxSlot}
  }
 ]
}
