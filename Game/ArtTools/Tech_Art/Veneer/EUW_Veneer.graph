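# Editor utility widget for the Veneer art tool, which creates and edits Veneer placement actors in the level.
# Add spawns a placement at a camera trace, facing and pinning any selected actors.
# Tick follows selection, updates container dynamic materials and rebuilds data-table-driven parameter widgets when veneer type changes.
# Tabs show placement settings, per-actor component details, and lists of in-use and missing materials with batch instance creation.
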EditorUtilityWidgetBlueprint /Game/ArtTools/Tech_Art/Veneer/EUW_Veneer
graphs [
 Actor Tab Text Color {nodes [
   Actor Tab Text Color() {is=title K2Node_FunctionEntry}
   n1 = Get ActorTab() -> (ActorTab:Editor Utility Check Box Object Reference)
   > Return Node!(
    execute
    ReturnValue:Slate Color Structure=Select(
     Option 0:Slate Color Structure=Get NoHover() -> (NoHover:Slate Color Structure)
     Option 1:Slate Color Structure=Get Hover() -> (Hover:Slate Color Structure)
     Index=IsHovered(self=@n1)
    ) -> (ReturnValue:Slate Color Structure)
   )
  ]}
 ApplyVeneerMaterials {nodes [
   ApplyVeneerMaterials() {is=title K2Node_FunctionEntry}
   > For Each Loop(Array:Array of Veneer Component Object References=Get VeneerComponents!(self:EBP Veneer Placement Object Reference=Get ActiveVeneer() -> (ActiveVeneer:EBP Veneer Placement Object Reference)) -> (VeneerComponents:Array of Veneer Component Object References)) -> (Array Element:Veneer Component Object Reference)
  ]}
 CheckForTypeChange {nodes [
   CheckForTypeChange() {is=title K2Node_FunctionEntry}
   n1 = Get VeneerType!(self:EBP Veneer Placement Object Reference=Get ActiveVeneer()) -> (VeneerType:Enum_VeneerType Enum=NewEnumerator0)
   > n2 = Branch(Condition="Equal (Enum)"(A:Enum_VeneerType Enum=@n1, B:Enum_VeneerType Enum=Get VeneerTypeBuffer() -> (VeneerTypeBuffer:Enum_VeneerType Enum=NewEnumerator0)))
   > n3 = Set Update(Update:bool=false) -> (Output_Get:bool)
   Return Node!(execute=[@n3.then, @n4.then], UpdateUI:bool=Get Update() -> (Update:bool))
   Set Update(execute=@n2.else, Update=true)
   > n4 = Set with Broadcast VeneerTypeBuffer!(execute, VeneerTypeBuffer:Enum_VeneerType Enum=@n1) -> (then, Output_Get:Enum_VeneerType Enum=NewEnumerator0)
  ]}
 ClearContainerMaps {nodes [
   ClearContainerMaps() {is=title K2Node_FunctionEntry}
   n1 = Get VeneerContainer!(self:EBP Veneer Placement Object Reference=Get ActiveVeneer()) -> (VeneerContainer:BP Veneer Container Object Reference)
   > Clear!(
    execute
    TargetMap:"Map of Names to Float (double-precision)s"=Get ScalarParams!(self:BP Veneer Container Object Reference=@n1) -> (ScalarParams:"Map of Names to Float (double-precision)s")
   ) -> (then) {is=title K2Node_CallFunction}
   > Clear(TargetMap:Map of Names to Linear Color Structures=Get VectorParams!(self:BP Veneer Container Object Reference=@n1) -> (VectorParams:Map of Names to Linear Color Structures))
   > Clear(TargetMap:Map of Names to Texture 2D Object References=Get TextureParamNames!(self:BP Veneer Container Object Reference=@n1) -> (TextureParamNames:Map of Names to Texture 2D Object References))
  ]}
 ConvertOriginalPathToVRN {nodes [
   n1 = Split(SourceString=Inputs!() -> (SourceString:String), InStr=/, SearchDir=FromEnd)
   n2 = "To Integer (Byte)"!(InByte:Byte=Get VeneerType(self=Get ActiveVeneer()) -> (VeneerType=NewEnumerator0)) -> (ReturnValue:int=0)
   n3 = Make Array!("[0]":String=D, "[1]":String=DA, "[2]":String=C, "[3]":String=L) -> (Array:Array of Strings)
   Outputs!(
    Asset Name:String=Append!(
     A=Split(SourceString=@n1.RightS, InStr=., SearchDir=FromEnd)
     B=_VNR_
     C:String="Get (a ref)"(Array:Array of Strings=@n3, Dimension 1=@n2) -> (Output:"String (by ref)")
    ) -> (ReturnValue)
    Asset Path:String=Append!(A=@n1, B=/VNR/) -> (ReturnValue)
   )
  ]}
 CreateMIC {nodes [
   n1 = CreateMIC!() -> (then, Material:Material Interface Object Reference) {is=title K2Node_FunctionEntry}
   > n2 = GetPathNameForLoadedAsset(LoadedAsset=@n1)
   > n3 = GetAssetTools()
   n4 = Convert Original Path to VRN!(SourceString:String=@n2) -> (Asset Name:String, Asset Path:String) {is=title K2Node_MacroInstance}
   > n5 = CreateAsset(self=@n3, AssetName=@n4, PackagePath=@n4.Asset Path, AssetClass=/Script/Engine.MaterialInstanceConstant)
   > n6 = Cast To MaterialInstanceConstant(Object=@n5)
   > n7 = Set LocalMIC(LocalMIC:Material Instance Constant Object Reference=@n6) -> (Output_Get:Material Instance Constant Object Reference)
   > SetMaterialInstanceParent!(execute, Instance:Material Instance Constant Object Reference=@n7, NewParent:Material Interface Object Reference=@n1) -> (then)
   > SetMaterialInstanceStaticSwitchParameterValue!(
    execute
    Instance:Material Instance Constant Object Reference=@n6
    ParameterName:Name=Enable Veneer
    Value:bool=true
    Association:EMaterialParameterAssociation Enum=GlobalParameter
    bUpdateMaterialInstance:bool=true
   ) -> (then, ReturnValue:bool)
   n8 = Get VeneerType(self=Get ActiveVeneer()) -> (VeneerType=NewEnumerator0)
   > SetMaterialInstanceStaticSwitchParameterValue(
    Instance=Get LocalMIC() -> (LocalMIC:Material Instance Constant Object Reference)
    ParameterName="Get (a ref)"(
     Array:Array of Names=Make Array("[0]":Name=Enable VNR Decal, "[1]":Name=Enable VNR Advanced Decal, "[2]":Name=Enable VNR Caustics, "[3]":Name=Enable VNR Layer) -> (Array:Array of Names)
     Dimension 1="To Integer (Byte)"(InByte=@n8) -> (ReturnValue=0)
    ) -> (Output:"Name (by ref)")
    Value=true
    Association=GlobalParameter
    bUpdateMaterialInstance=true
   )
  ]}
 CreateVeneerPlacementEntries {nodes [
   CreateVeneerPlacementEntries() {is=title K2Node_FunctionEntry}
   > ClearChildren(self=[
     Get CoordinatesPanel() -> (CoordinatesPanel:Vertical Box Object Reference)
     Get ParametersPanel() -> (ParametersPanel:Vertical Box Object Reference)
    ])
   n1 = Enum to String!(Enumerator:Enum_VeneerType Enum=Get VeneerType(self=Get ActiveVeneer()) -> (VeneerType=NewEnumerator0)) -> (ReturnValue:String)
   > n2 = Get Data Table Row Table_VeneerTypesParameters!(
    execute:Exec
    DataTable:Data Table Object Reference=/Game/ArtTools/Tech_Art/Veneer/Misc/Params/Table_VeneerTypesParameters
    RowName:Name=String To Name(InString=Replace(SourceString=@n1, From=" ", To=_))
   ) -> (then:Exec, RowNotFound:Exec, ReturnValue:Struct Veneer Struct Array Structure) {is=title K2Node_GetDataTableRow}
   > n3 = For Each Loop(Array:Array of Struct Veneer Parameters Structures=Break Struct Veneer Struct Array!(Struct_VeneerStructArray:"Struct Veneer Struct Array Structure (by ref)"=@n2) -> (Param_3_#1:Array of Struct Veneer Parameters Structures) {is=title K2Node_BreakStruct}) -> (Array Element:Struct Veneer Parameters Structure)
   > n4 = Set Param(Param:Struct Veneer Parameters Structure=@n3) -> (Output_Get:Struct Veneer Parameters Structure)
   n5 = Break Struct Veneer Parameters!(Struct_VeneerParameters:"Struct Veneer Parameters Structure (by ref)"=@n4) -> (ParamType_30_#2:Enum_VeneerParamTypes Enum, CreateEntry_40_#3:bool=~) {is=title K2Node_BreakStruct}
   > Branch(Condition=@n5.CreateEntry_40_#3)
   > n6 = Switch on Enum_VeneerParamTypes!(execute, Selection:Enum_VeneerParamTypes Enum=@n5) -> (NewEnumerator0:Exec, NewEnumerator1:Exec, NewEnumerator2:Exec, NewEnumerator3:Exec) {is=title K2Node_SwitchEnum}
   > Reroute Node()
   VeneerScalarEntry!(
    execute=@n6.NewEnumerator1
    self
    Veneer Param Struct:Struct Veneer Parameters Structure=Get Param() -> (Param:Struct Veneer Parameters Structure)
   ) -> (then)
   VeneerVectorEntry!(execute=@n6.NewEnumerator2, self, Veneer Param Struct:Struct Veneer Parameters Structure=Get Param()) -> (then)
   VeneerTextureEntry!(execute=@n6.NewEnumerator3, self, Veneer Param Struct:Struct Veneer Parameters Structure=Get Param()) -> (then)
   PrintString(execute=@n2.RowNotFound, InString=Please Enter Type)
  ]}
 DestroyMaterialsList {nodes [
   DestroyMaterialsList() {is=title K2Node_FunctionEntry}
   > ClearChildren(self=[
     Get MissingMaterialsPanel() -> (MissingMaterialsPanel:Vertical Box Object Reference)
     Get MaterialsInUsePanel() -> (MaterialsInUsePanel:Vertical Box Object Reference)
    ])
   > Clear(TargetMap:Map of Material Interface Object References to Booleans=Get Queue Missing Materials() -> (Queue Missing Materials:Map of Material Interface Object References to Booleans))
   > Clear(TargetArray:Array of Struct Veneer Original Materials Structures=Get MaterialsInUseBuffer() -> (MaterialsInUseBuffer:Array of Struct Veneer Original Materials Structures)) {is=title K2Node_CallArrayFunction}
  ]}
 EventGraph {nodes [
   "On Clicked (Btn_Add)"() {is=title K2Node_ComponentBoundEvent}
   > NewVeneerPlacement()
   > CreateVeneerPlacementEntries() {is=title K2Node_CallFunction}
   "On Check State Changed (PlacementTab)"!() -> (OutputDelegate, then, bIsChecked:bool) {is=title K2Node_ComponentBoundEvent}
   > TabsLogic!(
    execute
    self
    ItemToFind:"Editor Utility Check Box Object Reference (by ref)"=Get PlacementTab() -> (PlacementTab:Editor Utility Check Box Object Reference)
   ) -> (then)
   > DestroyMaterialsList() {is=title K2Node_CallFunction}
   "On Check State Changed (ActorTab)"!() -> (OutputDelegate, then, bIsChecked:bool) {is=title K2Node_ComponentBoundEvent}
   > TabsLogic(ItemToFind=Get ActorTab())
   > DestroyMaterialsList()
   "On Check State Changed (MaterialsTab)"!() -> (OutputDelegate, then, bIsChecked:bool) {is=title K2Node_ComponentBoundEvent}
   > TabsLogic(ItemToFind=Get MaterialsTab() -> (MaterialsTab:Editor Utility Check Box Object Reference))
   > PopulateMissingMateialsList()
   > PopulateMaterialsInUseList()
   "On Clicked (FocusButton)"() {is=title K2Node_ComponentBoundEvent}
   n1 = Get ActiveVeneer()
   > Is Valid(InputObject=@n1)
   > SetSelectedLevelActors(self=Get EditorActorSubsystem(), ActorsToSelect=Make Array!("[0]":Actor Object Reference=@n1) -> (Array:Array of Actor Object References))
   > ExecuteConsoleCommandFromEditor!(execute, ConsoleCommand:String=CAMERA ALIGN ACTIVEVIEWPORTONLY) -> (then)
   "On Clicked (CreateAllMissingMaterialsBtn)"() {is=title K2Node_ComponentBoundEvent}
   > n2 = For Each Loop(Array:Array of Widget Object References=GetAllChildren(self=Get MissingMaterialsPanel())) -> (Array Element:Widget Object Reference)
   Refresh Materials Lists(execute=@n2.Completed)
   n3 = Cast To EUW_MissingMaterialElement!(execute=@n2.LoopBody, Object=@n2) -> (then, CastFailed, AsEUW Missing Material Element:EUW Missing Material Element Object Reference) {is=title K2Node_DynamicCast}
   > Remove Create MIC(self:EUW Missing Material Element Object Reference=@n3)
   "On Clicked (PinActors)"() {is=title K2Node_ComponentBoundEvent}
   n4 = Get ActiveVeneer()
   > Is Valid(InputObject=@n4)
   > n5 = GetSelectedActors()
   > n6 = For Each Loop(Array:Array of Actor Object References=@n5) -> (Array Element:Actor Object Reference)
   > Is Valid(InputObject=GetComponentByClass!(self:Actor Object Reference=@n6, ComponentClass:Actor Component Class Reference=&StaticMeshComponent) -> (ReturnValue:Static Mesh Component Object Reference))
   > Add Unique(
    TargetArray:Array of Actor Object References=Get PinnedActors!(self:EBP Veneer Placement Object Reference=@n4) -> (PinnedActors:Array of Actor Object References)
    NewItem:"Actor Object Reference (by ref)"=@n6
   )
   "On Clicked (UnpinActors)"() {is=title K2Node_ComponentBoundEvent}
   n7 = Get ActiveVeneer()
   > Is Valid(InputObject=@n7)
   > n8 = GetSelectedActors()
   > n9 = For Each Loop(Array=@n8)
   > Remove Item!(execute, TargetArray:Array of Actor Object References=Get PinnedActors(self=@n7), Item:"Actor Object Reference (by ref)"=@n9) -> (then, ReturnValue:bool)
   Refresh Materials Lists() {is=title K2Node_CustomEvent}
   > DestroyMaterialsList()
   > PopulateMissingMateialsList()
   > PopulateMaterialsInUseList()
   Event Tick()
   > SelectVeneerPlacement()
   > Is Valid(InputObject=Get ActiveVeneer())
   > GetActorSettings()
   > UpdateVeneerContainerMIDs()
   > Branch(Condition=CheckForTypeChange!() -> (UpdateUI:bool) {is=title K2Node_CallFunction})
   > ClearContainerMaps() {is=title K2Node_CallFunction}
   > CreateVeneerPlacementEntries()
   Event Destruct()
   > BindOnDelete!(execute, self:EBP Veneer Placement Object Reference=Get ActiveVeneer(), Unbind:bool=true) -> (then)
   Get VeneerTypeBuffer() -> (VeneerTypeBuffer=NewEnumerator0)
  ]}
 GetActorSettings {nodes [
   GetActorSettings() {is=title K2Node_FunctionEntry}
   > n1 = GetSelectedActors()
   n2 = GetComponentByClass(
    self="Get (a copy)"(Array:Array of Actor Object References=@n1, Dimension 1=Last Index!(TargetArray:Array of Actor Object References=@n1) -> (ReturnValue=0)) -> (Output:Actor Object Reference)
    ComponentClass=&/VeneerComponent.VeneerComponent_C
   ) -> (ReturnValue:Veneer Component Object Reference)
   > Is Valid(InputObject=@n2)
   > SetObject(self=Get DetailsView_70() -> (DetailsView_70:Details View Object Reference), NewObject=@n2)
  ]}
 Get_ActorSettings_Visibility {nodes [
   Get_ActorSettings_Visibility() {is=title K2Node_FunctionEntry}
   n1 = Get ActorTab()
   > Return Node!(
    execute
    ReturnValue:ESlateVisibility Enum=Select(Option 0:ESlateVisibility Enum=Collapsed, Option 1:ESlateVisibility Enum=Visible, Index=IsChecked(self=@n1)) -> (ReturnValue:ESlateVisibility Enum)
   )
  ]}
 Get_MaterialsPanel_Visibility {nodes [
   Get_MaterialsPanel_Visibility() {is=title K2Node_FunctionEntry}
   n1 = Get MaterialsTab()
   > Return Node(ReturnValue=Select(Option 0=Collapsed, Option 1=Visible, Index=IsChecked(self=@n1)))
  ]}
 Get_MaterialsWarning_Visibility {nodes [
   Get_MaterialsWarning_Visibility() {is=title K2Node_FunctionEntry}
   n1 = Get ActiveVeneer()
   n2 = IsValid(Object=@n1)
   n3 = Is Not Empty!(TargetMap:Map of Material Interface Object References to Booleans=Get QueueMissingMaterials!(self:EBP Veneer Placement Object Reference=@n1) -> (QueueMissingMaterials:Map of Material Interface Object References to Booleans)) -> (ReturnValue:bool) {is=title K2Node_CallFunction}
   > Return Node(ReturnValue=Select(Option 0=Collapsed, Option 1=Visible, Index=AND Boolean(A=@n2, B=@n3)))
  ]}
 Get_MissingMaterialsBatchPanel_Visibility {nodes [
   Get_MissingMaterialsBatchPanel_Visibility() {is=title K2Node_FunctionEntry}
   n1 = Get Queue Missing Materials()
   > Return Node(ReturnValue=Select(
     Option 0=Visible
     Option 1=Collapsed
     Index=Is Empty!(TargetMap:Map of Material Interface Object References to Booleans=@n1) -> (ReturnValue:bool) {is=title K2Node_CallFunction}
    ))
  ]}
 Get_ObjectName_Text {nodes [
   Get_ObjectName_Text() {is=title K2Node_FunctionEntry}
   n1 = Get ActiveVeneer()
   n2 = GetDisplayName(Object=@n1)
   > Return Node(ReturnValue:Text=SelectText(A="To Text (String)"(InString=@n2), B="LOCTEXT(\"Select Veneer Actor\")", bPickA=IsValid(Object=@n1)))
  ]}
 Get_PlacementSettingsPanel_Visibility {nodes [
   Get_PlacementSettingsPanel_Visibility() {is=title K2Node_FunctionEntry}
   n1 = Get PlacementTab()
   > Return Node(ReturnValue:ESlateVisibility Enum=Select(Option 0=Collapsed, Option 1=Visible, Index=IsChecked(self=@n1)))
  ]}
 Get_SelectVeneerMessage_Visibility {nodes [
   Get_SelectVeneerMessage_Visibility() {is=title K2Node_FunctionEntry}
   n1 = Get ActiveVeneer()
   > Return Node(ReturnValue=Select(Option 0=Visible, Option 1=Collapsed, Index=IsValid(Object=@n1)))
  ]}
 Get_VeneerProperties_Visibility {nodes [
   Get_VeneerProperties_Visibility() {is=title K2Node_FunctionEntry}
   n1 = Get ActiveVeneer()
   > Return Node(ReturnValue=Select(Option 0=Hidden, Option 1=Visible, Index=IsValid(Object=@n1)))
  ]}
 Materials Tab Text Color {nodes [
   Materials Tab Text Color() {is=title K2Node_FunctionEntry}
   n1 = Get MaterialsTab()
   > Return Node(ReturnValue:Slate Color Structure=Select(Option 0:Slate Color Structure=Get NoHover(), Option 1:Slate Color Structure=Get Hover(), Index=IsHovered(self=@n1)) -> (ReturnValue:Slate Color Structure))
  ]}
 NewVeneerPlacement {nodes [
   NewVeneerPlacement() {is=title K2Node_FunctionEntry}
   > n1 = GetSelectedActors()
   > n2 = Branch(Condition="integer > integer"(A=Length(TargetArray=@n1)))
   > n3 = GetSelectedActors()
   > n4 = For Each Loop(Array=@n3)
   n5 = Length(TargetArray=@n3)
   Set OriginsCombined(
    execute=@n4.Completed
    OriginsCombined:Vector=vector / vector(A:Vector=Get OriginsCombined() -> (OriginsCombined:Vector), B:int=@n5) -> (ReturnValue:Vector)
   ) -> (Output_Get:Vector)
   > Set ForwardsCombined(ForwardsCombined:Vector=vector / vector(A=Get ForwardsCombined() -> (ForwardsCombined:Vector), B=@n5)) -> (Output_Get:Vector)
   n6 = GetLevelViewportCameraInfo!(self:Unreal Editor Subsystem Object Reference=Get UnrealEditorSubsystem()) -> (CameraLocation:Vector, CameraRotation:Rotator, ReturnValue:bool)
   n7 = Normalize!(A:Vector=vector - vector(A=Get OriginsCombined(), B=@n6), Tolerance:float=0.0001) -> (ReturnValue:Vector) {is=title K2Node_CallFunction}
   > n8 = Multi Line Trace For Objects!(
    execute
    WorldContextObject:Object Reference
    Start:Vector=@n6
    End:Vector=vector + vector(A=Get OriginsCombined(), B=vector * vector(A=@n7, B:float=5000) -> (ReturnValue=~))
    ObjectTypes:Array of EObjectTypeQuery Enums=Get Trace Object Types() -> (Trace Object Types:Array of EObjectTypeQuery Enums=ObjectTypeQuery1)
    bTraceComplex:bool=false
    ActorsToIgnore:Array of Actor Object References
    DrawDebugType:EDrawDebugTrace Enum
    bIgnoreSelf:bool=true
    TraceColor:Linear Color Structure=(R=1,G=0,B=0,A=1)
    TraceHitColor:Linear Color Structure=(R=0,G=1,B=0,A=1)
    DrawTime:float=5
   ) -> (then, OutHits:Array of Hit Result Structures, ReturnValue:bool)
   > n9 = For Each Loop with Break!(Exec:Exec, Array:Array of Hit Result Structures=@n8, Break:Exec=@n11.then_0) -> (LoopBody:Exec, Array Element:Hit Result Structure, Array Index:int, Completed:Exec)
   n10 = BreakHitResult!(Hit:"Hit Result Structure (by ref)"=@n9) -> (
    bBlockingHit:bool
    bInitialOverlap:bool
    Time:float
    Distance:float
    Location:Vector
    ImpactPoint:Vector
    Normal:Vector
    ImpactNormal:Vector
    PhysMat:Physical Material Object Reference
    HitActor:Actor Object Reference
    HitComponent:Primitive Component Object Reference
    HitBoneName:Name
    BoneName:Name
    HitItem:int
    ElementIndex:int
    FaceIndex:int
    TraceStart:Vector
    TraceEnd:Vector
   )
   > Branch(Condition=Contains Item(TargetArray:Array of Actor Object References=@n3, ItemToFind:"Actor Object Reference (by ref)"=@n10.HitActor))
   > n11 = Sequence()
   n12 = MakeRotFromX!(X:"Vector (by ref)"=vector * vector(A=@n10.ImpactNormal, B:double=-1) -> (ReturnValue=~)) -> (ReturnValue:Rotator)
   n13 = Get OriginsCombined()
   n14 = vector + vector(A=@n13, B=vector * vector(A:double=-10000, B:Vector=GetForwardVector!(InRot:Rotator=@n12) -> (ReturnValue:Vector)) -> (ReturnValue=~))
   n15 = Normalize(A=vector - vector(A=@n13, B=@n14), Tolerance=0.0001)
   n16 = Multi Line Trace For Objects(
    execute=@n11.then_1
    Start=@n14
    End=vector + vector(A=@n13, B=vector * vector(A:Vector=@n15, B:double=5000) -> (ReturnValue=~))
    ObjectTypes=Get Trace Object Types() -> (Trace Object Types=ObjectTypeQuery1)
    bTraceComplex=false
    bIgnoreSelf=true
    TraceColor=(R=1,G=0,B=0,A=1)
    TraceHitColor=(R=0,G=1,B=0,A=1)
    DrawTime=5
   )
   > n17 = For Each Loop with Break(Array=@n16, Break=@n19.then_0)
   n18 = BreakHitResult(Hit=@n17)
   > PrintString(InString=GetDisplayName(Object=@n18.HitActor))
   > Branch(Condition="Equal (Object)"(A=@n10.HitActor, B=@n18.HitActor))
   > n19 = Sequence()
   n20 = SpawnActorFromClass!(
    execute=@n19.then_1
    ActorClass:Actor Class Reference=&/EBP_VeneerPlacement.EBP_VeneerPlacement_C
    Location:Vector=vector + vector(A=@n18.Location, B=vector * vector(A=@n18.Normal, B=70) -> (ReturnValue=~))
    Rotation:Rotator=@n12
    bTransient:bool
   ) -> (then, ReturnValue:EBP Veneer Placement Object Reference)
   > n21 = Set ActiveVeneer(ActiveVeneer:EBP Veneer Placement Object Reference=@n20) -> (Output_Get:EBP Veneer Placement Object Reference)
   > n22 = GetSelectedActors()
   > Set PinnedActors!(execute, PinnedActors:Array of Actor Object References=@n22, self:EBP Veneer Placement Object Reference=@n21) -> (then, Output_Get:Array of Actor Object References)
   > SetSelectedLevelActors(self=Get EditorActorSubsystem(), ActorsToSelect=Make Array("[0]"=@n21))
   Set OriginsCombined(execute=@n4.LoopBody, OriginsCombined=vector + vector(A=Get Actor Bounds(self=@n4, bOnlyCollidingComponents=false), B=Get OriginsCombined()))
   > Set ForwardsCombined(ForwardsCombined=vector + vector(A=GetActorForwardVector!(self:Actor Object Reference=@n4) -> (ReturnValue:Vector), B=Get ForwardsCombined()))
   n23 = GetLevelViewportCameraInfo(self=Get UnrealEditorSubsystem())
   n24 = vector + vector(A=@n23, B=vector * vector(A=GetForwardVector(InRot=@n23.CameraRotation), B:Vector="1000,1000,1000") -> (ReturnValue=~))
   n25 = Line Trace By Channel!(
    execute=@n2.else
    WorldContextObject:Object Reference
    Start:Vector=@n23
    End:Vector=@n24
    TraceChannel:ETraceTypeQuery Enum
    bTraceComplex:bool=false
    ActorsToIgnore:Array of Actor Object References
    DrawDebugType:EDrawDebugTrace Enum
    bIgnoreSelf:bool
    TraceColor:Linear Color Structure
    TraceHitColor:Linear Color Structure
    DrawTime:float
   ) -> (then, OutHit:Hit Result Structure, ReturnValue:bool)
   > n26 = SpawnActorFromClass(
    ActorClass=&EBP_VeneerPlacement_C
    Location=SelectVector(A=BreakHitResult(Hit=@n25).Location, B=@n24, bPickA=@n25.ReturnValue)
   )
   > n27 = Set ActiveVeneer(ActiveVeneer=@n26)
   > SetSelectedLevelActors(self=Get EditorActorSubsystem(), ActorsToSelect=Make Array("[0]"=@n27))
   Contains Item(TargetArray=@n3, ItemToFind=@n18.HitActor)
   n28 = Length(TargetArray=@n3)
   vector / vector(A=Get ForwardsCombined(), B=@n28)
   vector / vector(A=@n13, B=@n28)
  ]}
 Panel Selection {nodes [
   n1 = Make Array!("[0]":String=Parameters, "[1]":String=Coordinates) -> (Array:Array of Strings)
   n2 = Inputs!() -> (ItemToFind:"String (by ref)")
   Outputs!(ReturnValue:Vertical Box Object Reference=Select(
     Option 0:Vertical Box Object Reference=Get ParametersPanel()
     Option 1:Vertical Box Object Reference=Get CoordinatesPanel()
     Index:int=Find Item!(TargetArray=@n1, ItemToFind:"String (by ref)"=@n2) -> (ReturnValue)
    ) -> (ReturnValue:Vertical Box Object Reference))
  ]}
 Placement Tab Text Color {nodes [
   Placement Tab Text Color() {is=title K2Node_FunctionEntry}
   n1 = Get PlacementTab()
   > Return Node(ReturnValue=Select(Option 0:Slate Color Structure=Get NoHover(), Option 1:Slate Color Structure=Get Hover(), Index:bool=IsHovered(self=@n1)) -> (ReturnValue:Slate Color Structure))
  ]}
 PopulateMaterialsInUseList {nodes [
   PopulateMaterialsInUseList() {is=title K2Node_FunctionEntry}
   n1 = Get VeneerBaseMaterials!(self:BP Veneer Container Object Reference=Get VeneerContainer(self=Get ActiveVeneer())) -> (VeneerBaseMaterials:Array of Struct Veneer Original Materials Structures)
   > n2 = Branch(Condition=Identical!(
     ArrayA:Array of Struct Veneer Original Materials Structures=@n1
     ArrayB:Array of Struct Veneer Original Materials Structures=Get MaterialsInUseBuffer()
    ) -> (ReturnValue:bool) {is=title K2Node_CallArrayFunction})
   Set MaterialsInUseBuffer(execute=@n2.else, MaterialsInUseBuffer:Array of Struct Veneer Original Materials Structures=@n1) -> (Output_Get:Array of Struct Veneer Original Materials Structures)
   > ClearChildren(self=Get MaterialsInUsePanel())
   n3 = Get ActiveVeneer()
   > n4 = For Each Loop(Array:Array of Struct Veneer Original Materials Structures=Get VeneerBaseMaterials(self=Get VeneerContainer(self=@n3))) -> (Array Element:Struct Veneer Original Materials Structure)
   > n5 = Create Widget!(
    execute:Exec
    WorldContextObject:Object Reference
    Class:User Widget Class Reference=/Game/ArtTools/Tech_Art/Veneer/SubWidgets/EUW_MaterialsInUse.EUW_MaterialsInUse_C
    OwningPlayer:Player Controller Object Reference
    Material:Material Interface Object Reference=Break Struct Veneer Original Materials!(Struct_VeneerOriginalMaterials:"Struct Veneer Original Materials Structure (by ref)"=@n4) -> (
     OriginalMaterial_2_#4:Material Interface Object Reference
     TextureData_8_#5:Array of Building Texture Data Object References
    ) {is=title K2Node_BreakStruct}
    IsSupported:bool=False
    VeneerWidget:EUW Veneer Object Reference
    IsDarkBG:bool="To Boolean (Integer)"!(InInt:int="% (Integer)"(A=@n4.Array Index, B=2) -> (ReturnValue=0)) -> (ReturnValue:bool)
   ) -> (then:Exec, ReturnValue:EUW Materials In Use Object Reference)
   > AddChild(self=Get MaterialsInUsePanel(), Content=@n5)
  ]}
 PopulateMissingMateialsList {nodes [
   PopulateMissingMateialsList() {is=title K2Node_FunctionEntry}
   > n1 = Sequence()
   n2 = Get QueueMissingMaterials(self=Get ActiveVeneer())
   n3 = Get Queue Missing Materials()
   > n4 = Branch(Condition="Equal (Integer)"(
     A=Length!(TargetMap:Map of Material Interface Object References to Booleans=@n2) -> (ReturnValue:int=0) {is=title K2Node_CallFunction}
     B=Length(TargetMap=@n3) -> (ReturnValue=0)
    ))
   n5 = Get QueueMissingMaterials(self=Get ActiveVeneer())
   > n6 = Keys!(execute, TargetMap:Map of Material Interface Object References to Booleans=@n5) -> (then, Keys:Array of Material Interface Object References)
   > n7 = Values!(execute, TargetMap:Map of Material Interface Object References to Booleans=@n5) -> (then, Values:Array of Booleans)
   n8 = Get Queue Missing Materials()
   > n9 = Keys(TargetMap=@n8)
   > n10 = Values(TargetMap=@n8)
   n11 = Identical(ArrayA:Array of Booleans=@n10, ArrayB:Array of Booleans=@n7)
   n12 = Identical(ArrayA:Array of Material Interface Object References=@n9, ArrayB:Array of Material Interface Object References=@n6)
   > Set Update(Update=NOT Boolean(A=AND Boolean(A=@n12, B=@n11)))
   Set Queue Missing Materials(execute=@n4.else, Queue Missing Materials:Map of Material Interface Object References to Booleans=@n2) -> (Output_Get:Map of Material Interface Object References to Booleans)
   > Set Update(Update=true)
   Branch(execute=@n1.then_1, Condition=Get Update())
   > ClearChildren(self=Get MissingMaterialsPanel())
   > n13 = Keys(TargetMap=Get Queue Missing Materials())
   > n14 = For Each Loop(Array:Array of Material Interface Object References=@n13) -> (Array Element:Material Interface Object Reference)
   > n15 = Values(TargetMap=Get Queue Missing Materials())
   > n16 = Create Widget(
    Class=/Game/ArtTools/Tech_Art/Veneer/SubWidgets/EUW_MissingMaterialElement.EUW_MissingMaterialElement_C
    Material=@n14
    IsSupported="Get (a ref)"(Array:Array of Booleans=@n15, Dimension 1=@n14.Array Index) -> (Output:"Boolean (by ref)")
    VeneerWidget=Self-Reference()
    IsDarkBG="To Boolean (Integer)"(InInt="% (Integer)"(A=@n14.Array Index, B=2) -> (ReturnValue=0))
   ) -> (ReturnValue:EUW Missing Material Element Object Reference)
   > AddChild(self=Get MissingMaterialsPanel(), Content=@n16)
  ]}
 SelectVeneerPlacement {nodes [
   SelectVeneerPlacement() {is=title K2Node_FunctionEntry}
   > n1 = GetSelectedActors()
   > n2 = For Each Loop with Break(Array:Array of Actor Object References=@n1, Break=@n7.then) -> (Array Element:Actor Object Reference)
   n3 = Branch(execute=@n2.Completed, Condition=Get QueueUpdate() -> (QueueUpdate:bool))
   > BindOnDelete(self=Get ActiveVeneer(), Unbind=false)
   > CreateVeneerPlacementEntries()
   n4 = Get ActiveVeneer()
   > SetObject(self=Get CustomPropertyView_46() -> (CustomPropertyView_46:Custom Property View Object Reference), NewObject=@n4)
   > Set with Broadcast VeneerTypeBuffer(VeneerTypeBuffer=Get VeneerType(self=@n4) -> (VeneerType=NewEnumerator0)) -> (Output_Get=NewEnumerator0)
   Update Veneer Placement(execute=@n3.else, self:EBP Veneer Placement Object Reference=Get ActiveVeneer())
   n5 = Cast To EBP_VeneerPlacement!(execute=@n2.LoopBody, Object=@n2) -> (then, CastFailed, AsEBP Veneer Placement:EBP Veneer Placement Object Reference) {is=title K2Node_DynamicCast}
   > n6 = Branch(Condition="Equal (Object)"(A=@n5, B=Get ActiveVeneer()))
   Set QueueUpdate(execute=@n6.else, QueueUpdate:bool=true) -> (Output_Get:bool)
   > BindOnDelete(self=Get ActiveVeneer(), Unbind=true)
   > n7 = Set ActiveVeneer(ActiveVeneer=@n5)
  ]}
 TabsLogic {nodes [
   n1 = TabsLogic!() -> (then, ItemToFind:"Editor Utility Check Box Object Reference (by ref)") {is=title K2Node_FunctionEntry}
   > n2 = For Each Loop with Break(
    Array:Array of Editor Utility Check Box Object References=Make Array!(
     "[0]":Editor Utility Check Box Object Reference=Get PlacementTab()
     "[1]":Editor Utility Check Box Object Reference=Get ActorTab()
     "[2]":Editor Utility Check Box Object Reference=Get MaterialsTab()
    ) -> (Array:Array of Editor Utility Check Box Object References)
    Break=@n4.then
   ) -> (Array Element:Editor Utility Check Box Object Reference)
   Set CheckedState!(execute=@n2.Completed, CheckedState:ECheckBoxState Enum=Checked, self:Check Box Object Reference=@n1) -> (then, Output_Get:ECheckBoxState Enum=Unchecked)
   n3 = Branch(execute=@n2.LoopBody, Condition="Equal (Object)"(A=@n2, B=@n1))
   Branch(execute=@n3.else, Condition=IsChecked(self=@n2))
   > n4 = SetIsChecked!(execute, self:Check Box Object Reference=@n2, InIsChecked:bool=false) -> (then)
  ]}
 UpdateVeneerContainerMIDs {nodes [
   UpdateVeneerContainerMIDs() {is=title K2Node_FunctionEntry}
   > Construct Dynamic Materials!(execute, self:BP Veneer Container Object Reference=Get VeneerContainer(self=Get ActiveVeneer()), TransformOnly:bool=false) -> (then)
  ]}
 VeneerScalarEntry {nodes [
   n1 = VeneerScalarEntry!() -> (then, Veneer Param Struct:Struct Veneer Parameters Structure) {is=title K2Node_FunctionEntry}
   n2 = Break Struct Veneer Parameters!(Struct_VeneerParameters=@n1) -> (
    ParamType_30_#2
    ParamName_31_#6:String
    FriendlyName_32_#7:String
    BooleanDefaultValue_33_#8:bool=~
    ScalarMinMax_34_#9:Vector 2D Structure
    ScalarStrictRangeValues_47_#10:bool=~
    ScalarDefaultValue_35_#11:double=~
    VectorDefaultValue_36_#12:Linear Color Structure
    TextueDefaultValue_37_#13:Texture 2D Object Reference
    CreateEntry_40_#3=~
    Catagory_43_#14:String
   )
   n3 = BreakVector2D(InVec=@n2.ScalarMinMax_34_#9)
   > n4 = Create Widget!(
    execute
    WorldContextObject
    Class=/Game/ArtTools/Tech_Art/Veneer/SubWidgets/EUW_VeneerScalarParam.EUW_VeneerScalarParam_C
    OwningPlayer
    FriendlyName:Name=String To Name(InString=@n2.FriendlyName_32_#7)
    ParameterName:Name=String To Name(InString=@n2.ParamName_31_#6)
    SpinnerMin:double=@n3
    SpinnerMax:double=@n3.Y
    Strict Min Max:bool=@n2.ScalarStrictRangeValues_47_#10
    Default Value:double=@n2.ScalarDefaultValue_35_#11
    Veneer Container:BP Veneer Container Object Reference=Get VeneerContainer(self=Get ActiveVeneer())
   ) -> (then, ReturnValue:EUW Veneer Scalar Param Object Reference)
   > AddChild(
    self=Panel Selection!(ItemToFind:"String (by ref)"=@n2.Catagory_43_#14) -> (ReturnValue:Vertical Box Object Reference) {is=title K2Node_MacroInstance}
    Content=@n4
   )
   > PrintString(InString=Kids are created)
  ]}
 VeneerTextureEntry {nodes [
   n1 = VeneerTextureEntry!() -> (then, Veneer Param Struct:Struct Veneer Parameters Structure) {is=title K2Node_FunctionEntry}
   n2 = Break Struct Veneer Parameters(Struct_VeneerParameters=@n1) -> (
    BooleanDefaultValue_33_#8=~
    ScalarStrictRangeValues_47_#10=~
    ScalarDefaultValue_35_#11=~
    CreateEntry_40_#3=~
   )
   > n3 = Create Widget!(
    execute
    WorldContextObject
    Class=/Game/ArtTools/Tech_Art/Veneer/SubWidgets/EUW_VeneerTextureParam.EUW_VeneerTextureParam_C
    OwningPlayer
    FriendlyName=String To Name(InString=@n2.FriendlyName_32_#7)
    ParameterName=String To Name(InString=@n2.ParamName_31_#6)
    Texture Value:Texture 2D Object Reference=@n2.TextueDefaultValue_37_#13
    Veneer Container=Get VeneerContainer(self=Get ActiveVeneer())
   ) -> (then, ReturnValue:EUW Veneer Texture Param Object Reference)
   > AddChild(self=Panel Selection(ItemToFind=@n2.Catagory_43_#14), Content=@n3)
  ]}
 VeneerVectorEntry {nodes [
   n1 = VeneerVectorEntry!() -> (then, Veneer Param Struct:Struct Veneer Parameters Structure) {is=title K2Node_FunctionEntry}
   n2 = Break Struct Veneer Parameters(Struct_VeneerParameters=@n1) -> (
    BooleanDefaultValue_33_#8=~
    ScalarStrictRangeValues_47_#10=~
    ScalarDefaultValue_35_#11=~
    CreateEntry_40_#3=~
   )
   > n3 = Create Widget!(
    execute
    WorldContextObject
    Class=/Game/ArtTools/Tech_Art/Veneer/SubWidgets/EUW_VeneerVectorParam.EUW_VeneerVectorParam_C
    OwningPlayer
    FriendlyName=String To Name(InString=@n2.FriendlyName_32_#7)
    ParameterName=String To Name(InString=@n2.ParamName_31_#6)
    Vector Value:Linear Color Structure=@n2.VectorDefaultValue_36_#12
    Veneer Container=Get VeneerContainer(self=Get ActiveVeneer())
   ) -> (then, ReturnValue:EUW Veneer Vector Param Object Reference)
   > AddChild(self=Panel Selection(ItemToFind=@n2.Catagory_43_#14), Content=@n3)
  ]}
]
variables [
 ActiveVeneer {type=EBP_VeneerPlacement_C, default=null}
 MaterialsInUseBuffer {type="array<Struct_VeneerOriginalMaterials>", default []}
 PinnedActorsBuffer {type="array<PrimitiveComponent>", default []}
 Queue Missing Materials {default {}}
 VeneerTypeBuffer {type=Enum_VeneerType, default=Decal}
]
widgets {
 info {rootWidgetClass=&CanvasPanel, widgetCount=44}
 widgets [
  CanvasPanel_0 {widgetClassPath=&CanvasPanel, widget=CanvasPanel_0}
  TopBox {
   widgetClassPath=&VerticalBox
   parent=CanvasPanel_0
   widget=TopBox
   slot {class=CanvasPanelSlot, properties {layoutData {alignment=<0 0>, anchors {maximum=<1 0>, minimum=<0 0>}, offsets=<ltrb 0 0 0 50>}}}
  }
  Header {
   widgetClassPath=&HorizontalBox
   parent=TopBox
   widget=Header
   slot {class=VerticalBoxSlot, properties {size {sizeRule=Fill, value=1}}}
  }
  FocusButton {
   widgetClassPath=&EditorUtilityButton
   parent=Header
   widget=FocusButton
   properties {widgetStyle {
     clickedSlateSound {resourceObject=null}
     disabled {
      bIsDynamicallyLoaded=false
      drawAs=Image
      imageType=FullColor
      mirroring=NoMirror
      resourceName=None
      resourceObject=/Game/ArtTools/Tech_Art/Veneer/Icons/Veneer_20
      tiling=NoTile
      imageSize=<20 20>
      margin=<ltrb 0 0 0 0>
      outlineSettings {
       bUseBrushTransparency=false
       roundingType=FixedRadius
       width=1
       color {colorUseRule=UseColor_ColorTable, specifiedColor=<rgba 0 0 0 0>}
       cornerRadii=<4 4 4 4>
      }
      tintColor {colorUseRule=UseColor_Specified, specifiedColor=<rgba 0.0625 0.0625 0.0625 1>}
      uVRegion {bIsValid=false, max=<0 0>, min=<0 0>}
     }
     disabledForeground {colorUseRule=UseColor_ColorTable, specifiedColor=<rgba 0 0 0 0>}
     hovered {
      bIsDynamicallyLoaded=false
      drawAs=Image
      imageType=FullColor
      mirroring=NoMirror
      resourceName=None
      resourceObject=&Veneer_20
      tiling=NoTile
      imageSize=<20 20>
      margin=<ltrb 0 0 0 0>
      outlineSettings {
       bUseBrushTransparency=false
       roundingType=FixedRadius
       width=1
       color {colorUseRule=UseColor_ColorTable, specifiedColor=<rgba 0 0 0 0>}
       cornerRadii=<4 4 4 4>
      }
      tintColor {colorUseRule=UseColor_Specified, specifiedColor=<rgba 1 1 1 1>}
      uVRegion {bIsValid=false, max=<0 0>, min=<0 0>}
     }
     hoveredForeground {colorUseRule=UseColor_ColorTable, specifiedColor=<rgba 0 0 0 0>}
     hoveredSlateSound {resourceObject=null}
     normal {
      bIsDynamicallyLoaded=false
      drawAs=Image
      imageType=FullColor
      mirroring=NoMirror
      resourceName=None
      resourceObject=&Veneer_20
      tiling=NoTile
      imageSize=<20 20>
      margin=<ltrb 0 0 0 0>
      outlineSettings {
       bUseBrushTransparency=false
       roundingType=FixedRadius
       width=1
       color {colorUseRule=UseColor_ColorTable, specifiedColor=<rgba 0 0 0 0>}
       cornerRadii=<4 4 4 4>
      }
      tintColor {colorUseRule=UseColor_Specified, specifiedColor=<rgba 0.7031 0.7031 0.7031 1>}
      uVRegion {bIsValid=false, max=<0 0>, min=<0 0>}
     }
     normalForeground {colorUseRule=UseColor_ColorTable, specifiedColor=<rgba 0 0 0 0>}
     normalPadding=<ltrb 0 0 0 0>
     pressed {
      bIsDynamicallyLoaded=false
      drawAs=Image
      imageType=FullColor
      mirroring=NoMirror
      resourceName=None
      resourceObject=&Veneer_20
      tiling=NoTile
      imageSize=<20 20>
      margin=<ltrb 0 0 0 0>
      outlineSettings {
       bUseBrushTransparency=false
       roundingType=FixedRadius
       width=1
       color {colorUseRule=UseColor_ColorTable, specifiedColor=<rgba 0 0 0 0>}
       cornerRadii=<4 4 4 4>
      }
      tintColor {colorUseRule=UseColor_Specified, specifiedColor=<rgba 0.4115 0.4115 0.4115 1>}
      uVRegion {bIsValid=false, max=<0 0>, min=<0 0>}
     }
     pressedForeground {colorUseRule=UseColor_ColorTable, specifiedColor=<rgba 0 0 0 0>}
     pressedPadding=<ltrb 0 0 0 0>
     pressedSlateSound {resourceObject=null}
    }}
   slot {class=HorizontalBoxSlot, properties {horizontalAlignment=HAlign_Center, verticalAlignment=VAlign_Center, padding=<ltrb 8 4 0 4>}}
  }
  Image_226 {
   widgetClassPath=&Image
   bIsVariable=true
   parent=FocusButton
   widget=Image_226
   properties {brush {
     bIsDynamicallyLoaded=false
     drawAs=Image
     imageType=FullColor
     mirroring=NoMirror
     resourceName=None
     resourceObject=&Veneer_20
     tiling=NoTile
     imageSize=<20 20>
     margin=<ltrb 0 0 0 0>
     outlineSettings {
      bUseBrushTransparency=false
      roundingType=HalfHeightRadius
      width=0
      color {colorUseRule=UseColor_Specified, specifiedColor=<rgba 0 0 0 0>}
      cornerRadii=<0 0 0 0>
     }
     tintColor {colorUseRule=UseColor_Specified, specifiedColor=<rgba 1 1 1 0>}
     uVRegion {bIsValid=false, max=<0 0>, min=<0 0>}
    }}
   slot {class=ButtonSlot, properties {padding=<ltrb 0 0 0 0>}}
  }
  ObjectName {
   widgetClassPath=&CommonTextBlock
   bIsVariable=true
   parent=Header
   widget=ObjectName
   properties {
    text=Veneer Actor
    colorAndOpacity {colorUseRule=UseColor_Specified, specifiedColor=<rgba 0.9647 0.9647 0.9647 1>}
    font {
     bForceMonospaced=false
     bMaterialIsStencil=false
     fontMaterial=null
     fontObject=&Roboto
     letterSpacing=0
     monospacedWidth=1
     size=11
     skewAmount=0
     typefaceFontName=Regular
     outlineSettings {
      bApplyOutlineToDropShadows=false
      bMiteredCorners=false
      bSeparateFillAlpha=false
      outlineMaterial=null
      outlineSize=0
      outlineColor=<rgba 0 0 0 1>
     }
    }
    shadowOffset=<0 0>
   }
   slot {class=HorizontalBoxSlot, properties {verticalAlignment=VAlign_Center, padding=<ltrb 8 0 0 0>}}
  }
  Btn_Add {
   widgetClassPath=&Button
   parent=Header
   widget=Btn_Add
   properties {widgetStyle {
     clickedSlateSound {resourceObject=null}
     disabled {
      bIsDynamicallyLoaded=false
      drawAs=RoundedBox
      imageType=NoImage
      mirroring=NoMirror
      resourceName=None
      resourceObject=null
      tiling=NoTile
      imageSize=<0 0>
      margin=<ltrb 0 0 0 0>
      outlineSettings {
       bUseBrushTransparency=false
       roundingType=FixedRadius
       width=1
       color {colorUseRule=UseColor_ColorTable, specifiedColor=<rgba 0 0 0 0>}
       cornerRadii=<4 4 4 4>
      }
      tintColor {colorUseRule=UseColor_Specified, specifiedColor=<rgba 0.0395 0.0395 0.0395 1>}
      uVRegion {bIsValid=false, max=<0 0>, min=<0 0>}
     }
     disabledForeground {colorUseRule=UseColor_ColorTable, specifiedColor=<rgba 0 0 0 0>}
     hovered {
      bIsDynamicallyLoaded=false
      drawAs=RoundedBox
      imageType=NoImage
      mirroring=NoMirror
      resourceName=None
      resourceObject=null
      tiling=NoTile
      imageSize=<32 32>
      margin=<ltrb 0 0 0 0>
      outlineSettings {
       bUseBrushTransparency=false
       roundingType=FixedRadius
       width=1
       color {colorUseRule=UseColor_ColorTable, specifiedColor=<rgba 0 0 0 0>}
       cornerRadii=<4 4 4 4>
      }
      tintColor {colorUseRule=UseColor_Specified, specifiedColor=<rgba 0.0953 0.0953 0.0953 1>}
      uVRegion {bIsValid=false, max=<0 0>, min=<0 0>}
     }
     hoveredForeground {colorUseRule=UseColor_ColorTable, specifiedColor=<rgba 0 0 0 0>}
     hoveredSlateSound {resourceObject=null}
     normal {
      bIsDynamicallyLoaded=false
      drawAs=RoundedBox
      imageType=NoImage
      mirroring=NoMirror
      resourceName=None
      resourceObject=null
      tiling=NoTile
      imageSize=<32 32>
      margin=<ltrb 0 0 0 0>
      outlineSettings {
       bUseBrushTransparency=false
       roundingType=FixedRadius
       width=1
       color {colorUseRule=UseColor_ColorTable, specifiedColor=<rgba 0 0 0 0>}
       cornerRadii=<4 4 4 4>
      }
      tintColor {colorUseRule=UseColor_Specified, specifiedColor=<rgba 0.0395 0.0395 0.0395 1>}
      uVRegion {bIsValid=false, max=<0 0>, min=<0 0>}
     }
     normalForeground {colorUseRule=UseColor_ColorTable, specifiedColor=<rgba 0 0 0 0>}
     normalPadding=<ltrb 12 1.5 12 1.5>
     pressed {
      bIsDynamicallyLoaded=false
      drawAs=RoundedBox
      imageType=NoImage
      mirroring=NoMirror
      resourceName=None
      resourceObject=null
      tiling=NoTile
      imageSize=<32 32>
      margin=<ltrb 0 0 0 0>
      outlineSettings {
       bUseBrushTransparency=false
       roundingType=FixedRadius
       width=1
       color {colorUseRule=UseColor_ColorTable, specifiedColor=<rgba 0 0 0 0>}
       cornerRadii=<4 4 4 4>
      }
      tintColor {colorUseRule=UseColor_Specified, specifiedColor=<rgba 0.0284 0.0284 0.0284 1>}
      uVRegion {bIsValid=false, max=<0 0>, min=<0 0>}
     }
     pressedForeground {colorUseRule=UseColor_ColorTable, specifiedColor=<rgba 0 0 0 0>}
     pressedPadding=<ltrb 12 2.5 12 0.5>
     pressedSlateSound {resourceObject=null}
    }}
   slot {class=HorizontalBoxSlot, properties {horizontalAlignment=HAlign_Right, padding=<ltrb 0 9 8 9>, size {sizeRule=Fill, value=1}}}
  }
  HorizontalBox_37 {
   widgetClassPath=&HorizontalBox
   parent=Btn_Add
   widget=HorizontalBox_37
   slot {class=ButtonSlot, properties {padding=<ltrb 4 5 4 5>}}
  }
  Image_71 {
   widgetClassPath=&Image
   bIsVariable=true
   parent=HorizontalBox_37
   widget=Image_71
   properties {brush {
     bIsDynamicallyLoaded=false
     drawAs=Image
     imageType=FullColor
     mirroring=NoMirror
     resourceName=None
     resourceObject=/Game/ArtTools/Tech_Art/Veneer/Icons/Veneer_14
     tiling=NoTile
     imageSize=<14 14>
     margin=<ltrb 0 0 0 0>
     outlineSettings {
      bUseBrushTransparency=false
      roundingType=HalfHeightRadius
      width=0
      color {colorUseRule=UseColor_Specified, specifiedColor=<rgba 0 0 0 0>}
      cornerRadii=<0 0 0 0>
     }
     tintColor {colorUseRule=UseColor_Specified, specifiedColor=<rgba 1 1 1 1>}
     uVRegion {bIsValid=false, max=<0 0>, min=<0 0>}
    }}
   slot {class=HorizontalBoxSlot, properties {horizontalAlignment=HAlign_Center, padding=<ltrb 0 0 3 0>}}
  }
  CommonTextBlock {
   widgetClassPath=&CommonTextBlock
   parent=HorizontalBox_37
   widget=CommonTextBlock
   properties {
    text=New
    font {
     bForceMonospaced=false
     bMaterialIsStencil=false
     fontMaterial=null
     fontObject=&Roboto
     letterSpacing=0
     monospacedWidth=1
     size=9
     skewAmount=0
     typefaceFontName=Regular
     outlineSettings {
      bApplyOutlineToDropShadows=false
      bMiteredCorners=false
      bSeparateFillAlpha=false
      outlineMaterial=null
      outlineSize=0
      outlineColor=<rgba 0 0 0 1>
     }
    }
    shadowOffset=<0 0>
   }
   slot {class=HorizontalBoxSlot, properties {horizontalAlignment=HAlign_Center, verticalAlignment=VAlign_Center}}
  }
  Border_0 {
   widgetClassPath=&Border
   parent=CanvasPanel_0
   widget=Border_0
   properties {
    background {
     bIsDynamicallyLoaded=false
     drawAs=Image
     imageType=NoImage
     mirroring=NoMirror
     resourceName=None
     resourceObject=null
     tiling=NoTile
     imageSize=<32 32>
     margin=<ltrb 0 0 0 0>
     outlineSettings {
      bUseBrushTransparency=false
      roundingType=HalfHeightRadius
      width=0
      color {colorUseRule=UseColor_Specified, specifiedColor=<rgba 0 0 0 0>}
      cornerRadii=<0 0 0 0>
     }
     tintColor {colorUseRule=UseColor_Specified, specifiedColor=<rgba 0.0075 0.0075 0.0075 1>}
     uVRegion {bIsValid=false, max=<0 0>, min=<0 0>}
    }
    padding=<ltrb 0 0 0 0>
   }
   slot {class=CanvasPanelSlot, properties {layoutData {alignment=<0 0>, anchors {maximum=<1 0>, minimum=<0 0>}, offsets=<ltrb 0 50 0 32>}}}
  }
  HorizontalBox_0 {
   widgetClassPath=&HorizontalBox
   parent=Border_0
   widget=HorizontalBox_0
   slot {class=BorderSlot, properties {padding=<ltrb 0 0 0 0>}}
  }
  PlacementTab {
   widgetClassPath=&EditorUtilityCheckBox
   bIsVariable=true
   parent=HorizontalBox_0
   widget=PlacementTab
   properties {
    checkedState=Checked
    widgetStyle {
     checkBoxType=ToggleButton
     backgroundHoveredImage {
      bIsDynamicallyLoaded=false
      drawAs=RoundedBox
      imageType=NoImage
      mirroring=NoMirror
      resourceName=None
      resourceObject=null
      tiling=NoTile
      imageSize=<18 18>
      margin=<ltrb 0 0 0 0>
      outlineSettings {
       bUseBrushTransparency=false
       roundingType=FixedRadius
       width=1
       color {colorUseRule=UseColor_ColorTable, specifiedColor=<rgba 0 0 0 0>}
       cornerRadii=<3 3 3 3>
      }
      tintColor {colorUseRule=UseColor_Specified, specifiedColor=<rgba 0.0048 0.0048 0.0048 1>}
      uVRegion {bIsValid=false, max=<0 0>, min=<0 0>}
     }
     backgroundImage {
      bIsDynamicallyLoaded=false
      drawAs=RoundedBox
      imageType=NoImage
      mirroring=NoMirror
      resourceName=None
      resourceObject=null
      tiling=NoTile
      imageSize=<18 18>
      margin=<ltrb 0 0 0 0>
      outlineSettings {
       bUseBrushTransparency=false
       roundingType=FixedRadius
       width=1
       color {colorUseRule=UseColor_ColorTable, specifiedColor=<rgba 0 0 0 0>}
       cornerRadii=<3 3 3 3>
      }
      tintColor {colorUseRule=UseColor_Specified, specifiedColor=<rgba 0.0048 0.0048 0.0048 1>}
      uVRegion {bIsValid=false, max=<0 0>, min=<0 0>}
     }
     backgroundPressedImage {
      bIsDynamicallyLoaded=false
      drawAs=RoundedBox
      imageType=NoImage
      mirroring=NoMirror
      resourceName=None
      resourceObject=null
      tiling=NoTile
      imageSize=<18 18>
      margin=<ltrb 0 0 0 0>
      outlineSettings {
       bUseBrushTransparency=false
       roundingType=FixedRadius
       width=1
       color {colorUseRule=UseColor_ColorTable, specifiedColor=<rgba 0 0 0 0>}
       cornerRadii=<3 3 3 3>
      }
      tintColor {colorUseRule=UseColor_Specified, specifiedColor=<rgba 0.0048 0.0048 0.0048 1>}
      uVRegion {bIsValid=false, max=<0 0>, min=<0 0>}
     }
     borderBackgroundColor {colorUseRule=UseColor_Specified, specifiedColor=<rgba 1 1 1 1>}
     checkedForeground {colorUseRule=UseColor_Specified, specifiedColor=<rgba 1 1 1 1>}
     checkedHoveredForeground {colorUseRule=UseColor_Specified, specifiedColor=<rgba 1 1 1 1>}
     checkedHoveredImage {
      bIsDynamicallyLoaded=false
      drawAs=RoundedBox
      imageType=NoImage
      mirroring=NoMirror
      resourceName=None
      resourceObject=null
      tiling=NoTile
      imageSize=<0 0>
      margin=<ltrb 0 0 0 0>
      outlineSettings {
       bUseBrushTransparency=false
       roundingType=FixedRadius
       width=0
       color {colorUseRule=UseColor_Specified, specifiedColor=<rgba 0 0 0 0>}
       cornerRadii=<3 3 0 0>
      }
      tintColor {colorUseRule=UseColor_Specified, specifiedColor=<rgba 0.0176 0.0176 0.0176 1>}
      uVRegion {bIsValid=false, max=<0 0>, min=<0 0>}
     }
     checkedImage {
      bIsDynamicallyLoaded=false
      drawAs=RoundedBox
      imageType=NoImage
      mirroring=NoMirror
      resourceName=None
      resourceObject=null
      tiling=NoTile
      imageSize=<0 0>
      margin=<ltrb 0 0 0 0>
      outlineSettings {
       bUseBrushTransparency=false
       roundingType=FixedRadius
       width=0
       color {colorUseRule=UseColor_Specified, specifiedColor=<rgba 0 0 0 0>}
       cornerRadii=<3 3 0 0>
      }
      tintColor {colorUseRule=UseColor_Specified, specifiedColor=<rgba 0.0176 0.0176 0.0176 1>}
      uVRegion {bIsValid=false, max=<0 0>, min=<0 0>}
     }
     checkedPressedForeground {colorUseRule=UseColor_Specified, specifiedColor=<rgba 1 1 1 1>}
     checkedPressedImage {
      bIsDynamicallyLoaded=false
      drawAs=RoundedBox
      imageType=NoImage
      mirroring=NoMirror
      resourceName=None
      resourceObject=null
      tiling=NoTile
      imageSize=<0 0>
      margin=<ltrb 0 0 0 0>
      outlineSettings {
       bUseBrushTransparency=false
       roundingType=FixedRadius
       width=0
       color {colorUseRule=UseColor_Specified, specifiedColor=<rgba 0 0 0 0>}
       cornerRadii=<3 3 0 0>
      }
      tintColor {colorUseRule=UseColor_Specified, specifiedColor=<rgba 0.0176 0.0176 0.0176 1>}
      uVRegion {bIsValid=false, max=<0 0>, min=<0 0>}
     }
     checkedSlateSound {resourceObject=null}
     foregroundColor {colorUseRule=UseColor_Specified, specifiedColor=<rgba 1 1 1 1>}
     hoveredForeground {colorUseRule=UseColor_Specified, specifiedColor=<rgba 1 1 1 1>}
     hoveredSlateSound {resourceObject=null}
     padding=<ltrb 2 0 0 0>
     pressedForeground {colorUseRule=UseColor_Specified, specifiedColor=<rgba 1 1 1 1>}
     uncheckedHoveredImage {
      bIsDynamicallyLoaded=false
      drawAs=RoundedBox
      imageType=NoImage
      mirroring=NoMirror
      resourceName=None
      resourceObject=null
      tiling=NoTile
      imageSize=<0 0>
      margin=<ltrb 0 0 0 0>
      outlineSettings {
       bUseBrushTransparency=false
       roundingType=FixedRadius
       width=0
       color {colorUseRule=UseColor_Specified, specifiedColor=<rgba 0 0 0 0>}
       cornerRadii=<3 3 0 0>
      }
      tintColor {colorUseRule=UseColor_Specified, specifiedColor=<rgba 0.0152 0.0152 0.0152 1>}
      uVRegion {bIsValid=false, max=<0 0>, min=<0 0>}
     }
     uncheckedImage {
      bIsDynamicallyLoaded=false
      drawAs=RoundedBox
      imageType=FullColor
      mirroring=NoMirror
      resourceName=None
      resourceObject=null
      tiling=NoTile
      imageSize=<256 256>
      margin=<ltrb 0 0 0 0>
      outlineSettings {
       bUseBrushTransparency=false
       roundingType=FixedRadius
       width=0
       color {colorUseRule=UseColor_Specified, specifiedColor=<rgba 0 0 0 0>}
       cornerRadii=<3 3 0 0>
      }
      tintColor {colorUseRule=UseColor_Specified, specifiedColor=<rgba 0.0075 0.0075 0.0075 1>}
      uVRegion {bIsValid=false, max=<0 0>, min=<0 0>}
     }
     uncheckedPressedImage {
      bIsDynamicallyLoaded=false
      drawAs=RoundedBox
      imageType=NoImage
      mirroring=NoMirror
      resourceName=None
      resourceObject=null
      tiling=NoTile
      imageSize=<0 0>
      margin=<ltrb 0 0 0 0>
      outlineSettings {
       bUseBrushTransparency=false
       roundingType=FixedRadius
       width=0
       color {colorUseRule=UseColor_Specified, specifiedColor=<rgba 0 0 0 0>}
       cornerRadii=<3 3 0 0>
      }
      tintColor {colorUseRule=UseColor_Specified, specifiedColor=<rgba 0.0152 0.0152 0.0152 1>}
      uVRegion {bIsValid=false, max=<0 0>, min=<0 0>}
     }
     uncheckedSlateSound {resourceObject=null}
     undeterminedForeground {colorUseRule=UseColor_Specified, specifiedColor=<rgba 1 1 1 1>}
     undeterminedHoveredImage {
      bIsDynamicallyLoaded=false
      drawAs=NoDrawType
      imageType=NoImage
      mirroring=NoMirror
      resourceName=None
      resourceObject=null
      tiling=NoTile
      imageSize=<0 0>
      margin=<ltrb 0 0 0 0>
      outlineSettings {
       bUseBrushTransparency=false
       roundingType=HalfHeightRadius
       width=0
       color {colorUseRule=UseColor_Specified, specifiedColor=<rgba 0 0 0 0>}
       cornerRadii=<0 0 0 0>
      }
      tintColor {colorUseRule=UseColor_Specified, specifiedColor=<rgba 1 1 1 1>}
      uVRegion {bIsValid=false, max=<0 0>, min=<0 0>}
     }
     undeterminedImage {
      bIsDynamicallyLoaded=false
      drawAs=NoDrawType
      imageType=NoImage
      mirroring=NoMirror
      resourceName=None
      resourceObject=null
      tiling=NoTile
      imageSize=<0 0>
      margin=<ltrb 0 0 0 0>
      outlineSettings {
       bUseBrushTransparency=false
       roundingType=HalfHeightRadius
       width=0
       color {colorUseRule=UseColor_Specified, specifiedColor=<rgba 0 0 0 0>}
       cornerRadii=<0 0 0 0>
      }
      tintColor {colorUseRule=UseColor_Specified, specifiedColor=<rgba 1 1 1 1>}
      uVRegion {bIsValid=false, max=<0 0>, min=<0 0>}
     }
     undeterminedPressedImage {
      bIsDynamicallyLoaded=false
      drawAs=NoDrawType
      imageType=NoImage
      mirroring=NoMirror
      resourceName=None
      resourceObject=null
      tiling=NoTile
      imageSize=<0 0>
      margin=<ltrb 0 0 0 0>
      outlineSettings {
       bUseBrushTransparency=false
       roundingType=HalfHeightRadius
       width=0
       color {colorUseRule=UseColor_Specified, specifiedColor=<rgba 0 0 0 0>}
       cornerRadii=<0 0 0 0>
      }
      tintColor {colorUseRule=UseColor_Specified, specifiedColor=<rgba 1 1 1 1>}
      uVRegion {bIsValid=false, max=<0 0>, min=<0 0>}
     }
    }
   }
   slot {class=HorizontalBoxSlot, properties {padding=<ltrb 10 7 0 0>}}
  }
  TextBlock_36 {
   widgetClassPath=&TextBlock
   parent=PlacementTab
   widget=TextBlock_36
   properties {
    text=Placement Settings
    colorAndOpacity {colorUseRule=UseColor_Specified, specifiedColor=<rgba 0.5271 0.5271 0.5271 1>}
    font {
     bForceMonospaced=false
     bMaterialIsStencil=false
     fontMaterial=null
     fontObject=&Roboto
     letterSpacing=0
     monospacedWidth=1
     size=9.75
     skewAmount=0
     typefaceFontName=Regular
     outlineSettings {
      bApplyOutlineToDropShadows=false
      bMiteredCorners=false
      bSeparateFillAlpha=false
      outlineMaterial=null
      outlineSize=0
      outlineColor=<rgba 0 0 0 1>
     }
    }
    margin=<ltrb 10 5 10 0>
    strikeBrush {
     bIsDynamicallyLoaded=false
     drawAs=Image
     imageType=NoImage
     mirroring=NoMirror
     resourceName=None
     resourceObject=null
     tiling=NoTile
     imageSize=<32 32>
     margin=<ltrb 0 0 0 0>
     outlineSettings {
      bUseBrushTransparency=false
      roundingType=HalfHeightRadius
      width=0
      color {colorUseRule=UseColor_Specified, specifiedColor=<rgba 0 0 0 0>}
      cornerRadii=<0 0 0 0>
     }
     tintColor {colorUseRule=UseColor_Specified, specifiedColor=<rgba 0.5271 0.5271 0.5271 1>}
     uVRegion {bIsValid=false, max=<0 0>, min=<0 0>}
    }
   }
   slot {class=PanelSlot}
  }
  ActorTab {
   widgetClassPath=&EditorUtilityCheckBox
   bIsVariable=true
   parent=HorizontalBox_0
   widget=ActorTab
   properties {widgetStyle {
     checkBoxType=ToggleButton
     backgroundHoveredImage {
      bIsDynamicallyLoaded=false
      drawAs=RoundedBox
      imageType=NoImage
      mirroring=NoMirror
      resourceName=None
      resourceObject=null
      tiling=NoTile
      imageSize=<18 18>
      margin=<ltrb 0 0 0 0>
      outlineSettings {
       bUseBrushTransparency=false
       roundingType=FixedRadius
       width=1
       color {colorUseRule=UseColor_ColorTable, specifiedColor=<rgba 0 0 0 0>}
       cornerRadii=<3 3 3 3>
      }
      tintColor {colorUseRule=UseColor_Specified, specifiedColor=<rgba 0.0048 0.0048 0.0048 1>}
      uVRegion {bIsValid=false, max=<0 0>, min=<0 0>}
     }
     backgroundImage {
      bIsDynamicallyLoaded=false
      drawAs=RoundedBox
      imageType=NoImage
      mirroring=NoMirror
      resourceName=None
      resourceObject=null
      tiling=NoTile
      imageSize=<18 18>
      margin=<ltrb 0 0 0 0>
      outlineSettings {
       bUseBrushTransparency=false
       roundingType=FixedRadius
       width=1
       color {colorUseRule=UseColor_ColorTable, specifiedColor=<rgba 0 0 0 0>}
       cornerRadii=<3 3 3 3>
      }
      tintColor {colorUseRule=UseColor_Specified, specifiedColor=<rgba 0.0048 0.0048 0.0048 1>}
      uVRegion {bIsValid=false, max=<0 0>, min=<0 0>}
     }
     backgroundPressedImage {
      bIsDynamicallyLoaded=false
      drawAs=RoundedBox
      imageType=NoImage
      mirroring=NoMirror
      resourceName=None
      resourceObject=null
      tiling=NoTile
      imageSize=<18 18>
      margin=<ltrb 0 0 0 0>
      outlineSettings {
       bUseBrushTransparency=false
       roundingType=FixedRadius
       width=1
       color {colorUseRule=UseColor_ColorTable, specifiedColor=<rgba 0 0 0 0>}
       cornerRadii=<3 3 3 3>
      }
      tintColor {colorUseRule=UseColor_Specified, specifiedColor=<rgba 0.0048 0.0048 0.0048 1>}
      uVRegion {bIsValid=false, max=<0 0>, min=<0 0>}
     }
     borderBackgroundColor {colorUseRule=UseColor_Specified, specifiedColor=<rgba 1 1 1 1>}
     checkedForeground {colorUseRule=UseColor_Specified, specifiedColor=<rgba 1 1 1 1>}
     checkedHoveredForeground {colorUseRule=UseColor_Specified, specifiedColor=<rgba 1 1 1 1>}
     checkedHoveredImage {
      bIsDynamicallyLoaded=false
      drawAs=RoundedBox
      imageType=NoImage
      mirroring=NoMirror
      resourceName=None
      resourceObject=null
      tiling=NoTile
      imageSize=<0 0>
      margin=<ltrb 0 0 0 0>
      outlineSettings {
       bUseBrushTransparency=false
       roundingType=FixedRadius
       width=0
       color {colorUseRule=UseColor_Specified, specifiedColor=<rgba 0 0 0 0>}
       cornerRadii=<3 3 0 0>
      }
      tintColor {colorUseRule=UseColor_Specified, specifiedColor=<rgba 0.0176 0.0176 0.0176 1>}
      uVRegion {bIsValid=false, max=<0 0>, min=<0 0>}
     }
     checkedImage {
      bIsDynamicallyLoaded=false
      drawAs=RoundedBox
      imageType=NoImage
      mirroring=NoMirror
      resourceName=None
      resourceObject=null
      tiling=NoTile
      imageSize=<0 0>
      margin=<ltrb 0 0 0 0>
      outlineSettings {
       bUseBrushTransparency=false
       roundingType=FixedRadius
       width=0
       color {colorUseRule=UseColor_Specified, specifiedColor=<rgba 0 0 0 0>}
       cornerRadii=<3 3 0 0>
      }
      tintColor {colorUseRule=UseColor_Specified, specifiedColor=<rgba 0.0176 0.0176 0.0176 1>}
      uVRegion {bIsValid=false, max=<0 0>, min=<0 0>}
     }
     checkedPressedForeground {colorUseRule=UseColor_Specified, specifiedColor=<rgba 1 1 1 1>}
     checkedPressedImage {
      bIsDynamicallyLoaded=false
      drawAs=RoundedBox
      imageType=NoImage
      mirroring=NoMirror
      resourceName=None
      resourceObject=null
      tiling=NoTile
      imageSize=<0 0>
      margin=<ltrb 0 0 0 0>
      outlineSettings {
       bUseBrushTransparency=false
       roundingType=FixedRadius
       width=0
       color {colorUseRule=UseColor_Specified, specifiedColor=<rgba 0 0 0 0>}
       cornerRadii=<3 3 0 0>
      }
      tintColor {colorUseRule=UseColor_Specified, specifiedColor=<rgba 0.0176 0.0176 0.0176 1>}
      uVRegion {bIsValid=false, max=<0 0>, min=<0 0>}
     }
     checkedSlateSound {resourceObject=null}
     foregroundColor {colorUseRule=UseColor_Specified, specifiedColor=<rgba 1 1 1 1>}
     hoveredForeground {colorUseRule=UseColor_Specified, specifiedColor=<rgba 1 1 1 1>}
     hoveredSlateSound {resourceObject=null}
     padding=<ltrb 2 0 0 0>
     pressedForeground {colorUseRule=UseColor_Specified, specifiedColor=<rgba 1 1 1 1>}
     uncheckedHoveredImage {
      bIsDynamicallyLoaded=false
      drawAs=RoundedBox
      imageType=NoImage
      mirroring=NoMirror
      resourceName=None
      resourceObject=null
      tiling=NoTile
      imageSize=<0 0>
      margin=<ltrb 0 0 0 0>
      outlineSettings {
       bUseBrushTransparency=false
       roundingType=FixedRadius
       width=0
       color {colorUseRule=UseColor_Specified, specifiedColor=<rgba 0 0 0 0>}
       cornerRadii=<3 3 0 0>
      }
      tintColor {colorUseRule=UseColor_Specified, specifiedColor=<rgba 0.0152 0.0152 0.0152 1>}
      uVRegion {bIsValid=false, max=<0 0>, min=<0 0>}
     }
     uncheckedImage {
      bIsDynamicallyLoaded=false
      drawAs=RoundedBox
      imageType=FullColor
      mirroring=NoMirror
      resourceName=None
      resourceObject=null
      tiling=NoTile
      imageSize=<256 256>
      margin=<ltrb 0 0 0 0>
      outlineSettings {
       bUseBrushTransparency=false
       roundingType=FixedRadius
       width=0
       color {colorUseRule=UseColor_Specified, specifiedColor=<rgba 0 0 0 0>}
       cornerRadii=<3 3 0 0>
      }
      tintColor {colorUseRule=UseColor_Specified, specifiedColor=<rgba 0.0075 0.0075 0.0075 1>}
      uVRegion {bIsValid=false, max=<0 0>, min=<0 0>}
     }
     uncheckedPressedImage {
      bIsDynamicallyLoaded=false
      drawAs=RoundedBox
      imageType=NoImage
      mirroring=NoMirror
      resourceName=None
      resourceObject=null
      tiling=NoTile
      imageSize=<0 0>
      margin=<ltrb 0 0 0 0>
      outlineSettings {
       bUseBrushTransparency=false
       roundingType=FixedRadius
       width=0
       color {colorUseRule=UseColor_Specified, specifiedColor=<rgba 0 0 0 0>}
       cornerRadii=<3 3 0 0>
      }
      tintColor {colorUseRule=UseColor_Specified, specifiedColor=<rgba 0.0152 0.0152 0.0152 1>}
      uVRegion {bIsValid=false, max=<0 0>, min=<0 0>}
     }
     uncheckedSlateSound {resourceObject=null}
     undeterminedForeground {colorUseRule=UseColor_Specified, specifiedColor=<rgba 1 1 1 1>}
     undeterminedHoveredImage {
      bIsDynamicallyLoaded=false
      drawAs=NoDrawType
      imageType=NoImage
      mirroring=NoMirror
      resourceName=None
      resourceObject=null
      tiling=NoTile
      imageSize=<0 0>
      margin=<ltrb 0 0 0 0>
      outlineSettings {
       bUseBrushTransparency=false
       roundingType=HalfHeightRadius
       width=0
       color {colorUseRule=UseColor_Specified, specifiedColor=<rgba 0 0 0 0>}
       cornerRadii=<0 0 0 0>
      }
      tintColor {colorUseRule=UseColor_Specified, specifiedColor=<rgba 1 1 1 1>}
      uVRegion {bIsValid=false, max=<0 0>, min=<0 0>}
     }
     undeterminedImage {
      bIsDynamicallyLoaded=false
      drawAs=NoDrawType
      imageType=NoImage
      mirroring=NoMirror
      resourceName=None
      resourceObject=null
      tiling=NoTile
      imageSize=<0 0>
      margin=<ltrb 0 0 0 0>
      outlineSettings {
       bUseBrushTransparency=false
       roundingType=HalfHeightRadius
       width=0
       color {colorUseRule=UseColor_Specified, specifiedColor=<rgba 0 0 0 0>}
       cornerRadii=<0 0 0 0>
      }
      tintColor {colorUseRule=UseColor_Specified, specifiedColor=<rgba 1 1 1 1>}
      uVRegion {bIsValid=false, max=<0 0>, min=<0 0>}
     }
     undeterminedPressedImage {
      bIsDynamicallyLoaded=false
      drawAs=NoDrawType
      imageType=NoImage
      mirroring=NoMirror
      resourceName=None
      resourceObject=null
      tiling=NoTile
      imageSize=<0 0>
      margin=<ltrb 0 0 0 0>
      outlineSettings {
       bUseBrushTransparency=false
       roundingType=HalfHeightRadius
       width=0
       color {colorUseRule=UseColor_Specified, specifiedColor=<rgba 0 0 0 0>}
       cornerRadii=<0 0 0 0>
      }
      tintColor {colorUseRule=UseColor_Specified, specifiedColor=<rgba 1 1 1 1>}
      uVRegion {bIsValid=false, max=<0 0>, min=<0 0>}
     }
    }}
   slot {class=HorizontalBoxSlot, properties {padding=<ltrb 10 7 0 0>}}
  }
  TextBlock {
   widgetClassPath=&TextBlock
   parent=ActorTab
   widget=TextBlock
   properties {
    text=Actor Settings
    colorAndOpacity {colorUseRule=UseColor_Specified, specifiedColor=<rgba 0.5271 0.5271 0.5271 1>}
    font {
     bForceMonospaced=false
     bMaterialIsStencil=false
     fontMaterial=null
     fontObject=&Roboto
     letterSpacing=0
     monospacedWidth=1
     size=9.75
     skewAmount=0
     typefaceFontName=Regular
     outlineSettings {
      bApplyOutlineToDropShadows=false
      bMiteredCorners=false
      bSeparateFillAlpha=false
      outlineMaterial=null
      outlineSize=0
      outlineColor=<rgba 0 0 0 1>
     }
    }
    margin=<ltrb 10 5 10 0>
    strikeBrush {
     bIsDynamicallyLoaded=false
     drawAs=Image
     imageType=NoImage
     mirroring=NoMirror
     resourceName=None
     resourceObject=null
     tiling=NoTile
     imageSize=<32 32>
     margin=<ltrb 0 0 0 0>
     outlineSettings {
      bUseBrushTransparency=false
      roundingType=HalfHeightRadius
      width=0
      color {colorUseRule=UseColor_Specified, specifiedColor=<rgba 0 0 0 0>}
      cornerRadii=<0 0 0 0>
     }
     tintColor {colorUseRule=UseColor_Specified, specifiedColor=<rgba 0.5271 0.5271 0.5271 1>}
     uVRegion {bIsValid=false, max=<0 0>, min=<0 0>}
    }
   }
   slot {class=PanelSlot}
  }
  MaterialsTab {
   widgetClassPath=&EditorUtilityCheckBox
   bIsVariable=true
   parent=HorizontalBox_0
   widget=MaterialsTab
   properties {widgetStyle {
     checkBoxType=ToggleButton
     backgroundHoveredImage {
      bIsDynamicallyLoaded=false
      drawAs=RoundedBox
      imageType=NoImage
      mirroring=NoMirror
      resourceName=None
      resourceObject=null
      tiling=NoTile
      imageSize=<18 18>
      margin=<ltrb 0 0 0 0>
      outlineSettings {
       bUseBrushTransparency=false
       roundingType=FixedRadius
       width=1
       color {colorUseRule=UseColor_ColorTable, specifiedColor=<rgba 0 0 0 0>}
       cornerRadii=<3 3 3 3>
      }
      tintColor {colorUseRule=UseColor_Specified, specifiedColor=<rgba 0.0048 0.0048 0.0048 1>}
      uVRegion {bIsValid=false, max=<0 0>, min=<0 0>}
     }
     backgroundImage {
      bIsDynamicallyLoaded=false
      drawAs=RoundedBox
      imageType=NoImage
      mirroring=NoMirror
      resourceName=None
      resourceObject=null
      tiling=NoTile
      imageSize=<18 18>
      margin=<ltrb 0 0 0 0>
      outlineSettings {
       bUseBrushTransparency=false
       roundingType=FixedRadius
       width=1
       color {colorUseRule=UseColor_ColorTable, specifiedColor=<rgba 0 0 0 0>}
       cornerRadii=<3 3 3 3>
      }
      tintColor {colorUseRule=UseColor_Specified, specifiedColor=<rgba 0.0048 0.0048 0.0048 1>}
      uVRegion {bIsValid=false, max=<0 0>, min=<0 0>}
     }
     backgroundPressedImage {
      bIsDynamicallyLoaded=false
      drawAs=RoundedBox
      imageType=NoImage
      mirroring=NoMirror
      resourceName=None
      resourceObject=null
      tiling=NoTile
      imageSize=<18 18>
      margin=<ltrb 0 0 0 0>
      outlineSettings {
       bUseBrushTransparency=false
       roundingType=FixedRadius
       width=1
       color {colorUseRule=UseColor_ColorTable, specifiedColor=<rgba 0 0 0 0>}
       cornerRadii=<3 3 3 3>
      }
      tintColor {colorUseRule=UseColor_Specified, specifiedColor=<rgba 0.0048 0.0048 0.0048 1>}
      uVRegion {bIsValid=false, max=<0 0>, min=<0 0>}
     }
     borderBackgroundColor {colorUseRule=UseColor_Specified, specifiedColor=<rgba 1 1 1 1>}
     checkedForeground {colorUseRule=UseColor_Specified, specifiedColor=<rgba 1 1 1 1>}
     checkedHoveredForeground {colorUseRule=UseColor_Specified, specifiedColor=<rgba 1 1 1 1>}
     checkedHoveredImage {
      bIsDynamicallyLoaded=false
      drawAs=RoundedBox
      imageType=NoImage
      mirroring=NoMirror
      resourceName=None
      resourceObject=null
      tiling=NoTile
      imageSize=<0 0>
      margin=<ltrb 0 0 0 0>
      outlineSettings {
       bUseBrushTransparency=false
       roundingType=FixedRadius
       width=0
       color {colorUseRule=UseColor_Specified, specifiedColor=<rgba 0 0 0 0>}
       cornerRadii=<3 3 0 0>
      }
      tintColor {colorUseRule=UseColor_Specified, specifiedColor=<rgba 0.0176 0.0176 0.0176 1>}
      uVRegion {bIsValid=false, max=<0 0>, min=<0 0>}
     }
     checkedImage {
      bIsDynamicallyLoaded=false
      drawAs=RoundedBox
      imageType=NoImage
      mirroring=NoMirror
      resourceName=None
      resourceObject=null
      tiling=NoTile
      imageSize=<0 0>
      margin=<ltrb 0 0 0 0>
      outlineSettings {
       bUseBrushTransparency=false
       roundingType=FixedRadius
       width=0
       color {colorUseRule=UseColor_Specified, specifiedColor=<rgba 0 0 0 0>}
       cornerRadii=<3 3 0 0>
      }
      tintColor {colorUseRule=UseColor_Specified, specifiedColor=<rgba 0.0176 0.0176 0.0176 1>}
      uVRegion {bIsValid=false, max=<0 0>, min=<0 0>}
     }
     checkedPressedForeground {colorUseRule=UseColor_Specified, specifiedColor=<rgba 1 1 1 1>}
     checkedPressedImage {
      bIsDynamicallyLoaded=false
      drawAs=RoundedBox
      imageType=NoImage
      mirroring=NoMirror
      resourceName=None
      resourceObject=null
      tiling=NoTile
      imageSize=<0 0>
      margin=<ltrb 0 0 0 0>
      outlineSettings {
       bUseBrushTransparency=false
       roundingType=FixedRadius
       width=0
       color {colorUseRule=UseColor_Specified, specifiedColor=<rgba 0 0 0 0>}
       cornerRadii=<3 3 0 0>
      }
      tintColor {colorUseRule=UseColor_Specified, specifiedColor=<rgba 0.0176 0.0176 0.0176 1>}
      uVRegion {bIsValid=false, max=<0 0>, min=<0 0>}
     }
     checkedSlateSound {resourceObject=null}
     foregroundColor {colorUseRule=UseColor_Specified, specifiedColor=<rgba 1 1 1 1>}
     hoveredForeground {colorUseRule=UseColor_Specified, specifiedColor=<rgba 1 1 1 1>}
     hoveredSlateSound {resourceObject=null}
     padding=<ltrb 2 0 0 0>
     pressedForeground {colorUseRule=UseColor_Specified, specifiedColor=<rgba 1 1 1 1>}
     uncheckedHoveredImage {
      bIsDynamicallyLoaded=false
      drawAs=RoundedBox
      imageType=NoImage
      mirroring=NoMirror
      resourceName=None
      resourceObject=null
      tiling=NoTile
      imageSize=<0 0>
      margin=<ltrb 0 0 0 0>
      outlineSettings {
       bUseBrushTransparency=false
       roundingType=FixedRadius
       width=0
       color {colorUseRule=UseColor_Specified, specifiedColor=<rgba 0 0 0 0>}
       cornerRadii=<3 3 0 0>
      }
      tintColor {colorUseRule=UseColor_Specified, specifiedColor=<rgba 0.0152 0.0152 0.0152 1>}
      uVRegion {bIsValid=false, max=<0 0>, min=<0 0>}
     }
     uncheckedImage {
      bIsDynamicallyLoaded=false
      drawAs=RoundedBox
      imageType=FullColor
      mirroring=NoMirror
      resourceName=None
      resourceObject=null
      tiling=NoTile
      imageSize=<256 256>
      margin=<ltrb 0 0 0 0>
      outlineSettings {
       bUseBrushTransparency=false
       roundingType=FixedRadius
       width=0
       color {colorUseRule=UseColor_Specified, specifiedColor=<rgba 0 0 0 0>}
       cornerRadii=<3 3 0 0>
      }
      tintColor {colorUseRule=UseColor_Specified, specifiedColor=<rgba 0.0075 0.0075 0.0075 1>}
      uVRegion {bIsValid=false, max=<0 0>, min=<0 0>}
     }
     uncheckedPressedImage {
      bIsDynamicallyLoaded=false
      drawAs=RoundedBox
      imageType=NoImage
      mirroring=NoMirror
      resourceName=None
      resourceObject=null
      tiling=NoTile
      imageSize=<0 0>
      margin=<ltrb 0 0 0 0>
      outlineSettings {
       bUseBrushTransparency=false
       roundingType=FixedRadius
       width=0
       color {colorUseRule=UseColor_Specified, specifiedColor=<rgba 0 0 0 0>}
       cornerRadii=<3 3 0 0>
      }
      tintColor {colorUseRule=UseColor_Specified, specifiedColor=<rgba 0.0152 0.0152 0.0152 1>}
      uVRegion {bIsValid=false, max=<0 0>, min=<0 0>}
     }
     uncheckedSlateSound {resourceObject=null}
     undeterminedForeground {colorUseRule=UseColor_Specified, specifiedColor=<rgba 1 1 1 1>}
     undeterminedHoveredImage {
      bIsDynamicallyLoaded=false
      drawAs=NoDrawType
      imageType=NoImage
      mirroring=NoMirror
      resourceName=None
      resourceObject=null
      tiling=NoTile
      imageSize=<0 0>
      margin=<ltrb 0 0 0 0>
      outlineSettings {
       bUseBrushTransparency=false
       roundingType=HalfHeightRadius
       width=0
       color {colorUseRule=UseColor_Specified, specifiedColor=<rgba 0 0 0 0>}
       cornerRadii=<0 0 0 0>
      }
      tintColor {colorUseRule=UseColor_Specified, specifiedColor=<rgba 1 1 1 1>}
      uVRegion {bIsValid=false, max=<0 0>, min=<0 0>}
     }
     undeterminedImage {
      bIsDynamicallyLoaded=false
      drawAs=NoDrawType
      imageType=NoImage
      mirroring=NoMirror
      resourceName=None
      resourceObject=null
      tiling=NoTile
      imageSize=<0 0>
      margin=<ltrb 0 0 0 0>
      outlineSettings {
       bUseBrushTransparency=false
       roundingType=HalfHeightRadius
       width=0
       color {colorUseRule=UseColor_Specified, specifiedColor=<rgba 0 0 0 0>}
       cornerRadii=<0 0 0 0>
      }
      tintColor {colorUseRule=UseColor_Specified, specifiedColor=<rgba 1 1 1 1>}
      uVRegion {bIsValid=false, max=<0 0>, min=<0 0>}
     }
     undeterminedPressedImage {
      bIsDynamicallyLoaded=false
      drawAs=NoDrawType
      imageType=NoImage
      mirroring=NoMirror
      resourceName=None
      resourceObject=null
      tiling=NoTile
      imageSize=<0 0>
      margin=<ltrb 0 0 0 0>
      outlineSettings {
       bUseBrushTransparency=false
       roundingType=HalfHeightRadius
       width=0
       color {colorUseRule=UseColor_Specified, specifiedColor=<rgba 0 0 0 0>}
       cornerRadii=<0 0 0 0>
      }
      tintColor {colorUseRule=UseColor_Specified, specifiedColor=<rgba 1 1 1 1>}
      uVRegion {bIsValid=false, max=<0 0>, min=<0 0>}
     }
    }}
   slot {class=HorizontalBoxSlot, properties {padding=<ltrb 10 7 0 0>}}
  }
  HorizontalBox_1 {widgetClassPath=&HorizontalBox, parent=MaterialsTab, widget=HorizontalBox_1, slot {class=PanelSlot}}
  TextBlock_1 {
   widgetClassPath=&TextBlock
   parent=HorizontalBox_1
   widget=TextBlock_1
   properties {
    text=Materials
    colorAndOpacity {colorUseRule=UseColor_Specified, specifiedColor=<rgba 0.5271 0.5271 0.5271 1>}
    font {
     bForceMonospaced=false
     bMaterialIsStencil=false
     fontMaterial=null
     fontObject=&Roboto
     letterSpacing=0
     monospacedWidth=1
     size=9.75
     skewAmount=0
     typefaceFontName=Regular
     outlineSettings {
      bApplyOutlineToDropShadows=false
      bMiteredCorners=false
      bSeparateFillAlpha=false
      outlineMaterial=null
      outlineSize=0
      outlineColor=<rgba 0 0 0 1>
     }
    }
    margin=<ltrb 10 5 10 0>
    strikeBrush {
     bIsDynamicallyLoaded=false
     drawAs=Image
     imageType=NoImage
     mirroring=NoMirror
     resourceName=None
     resourceObject=null
     tiling=NoTile
     imageSize=<32 32>
     margin=<ltrb 0 0 0 0>
     outlineSettings {
      bUseBrushTransparency=false
      roundingType=HalfHeightRadius
      width=0
      color {colorUseRule=UseColor_Specified, specifiedColor=<rgba 0 0 0 0>}
      cornerRadii=<0 0 0 0>
     }
     tintColor {colorUseRule=UseColor_Specified, specifiedColor=<rgba 0.5271 0.5271 0.5271 1>}
     uVRegion {bIsValid=false, max=<0 0>, min=<0 0>}
    }
   }
   slot {class=HorizontalBoxSlot}
  }
  Image {
   widgetClassPath=&Image
   bIsVariable=true
   parent=HorizontalBox_1
   widget=Image
   properties {
    visibility=Collapsed
    brush {
     bIsDynamicallyLoaded=false
     drawAs=Image
     imageType=FullColor
     mirroring=NoMirror
     resourceName=None
     resourceObject=/Game/UI/Foundation/Shell/Textures/T-Icon-Warning-32
     tiling=NoTile
     imageSize=<16 16>
     margin=<ltrb 0 0 0 0>
     outlineSettings {
      bUseBrushTransparency=false
      roundingType=HalfHeightRadius
      width=0
      color {colorUseRule=UseColor_Specified, specifiedColor=<rgba 0 0 0 0>}
      cornerRadii=<0 0 0 0>
     }
     tintColor {colorUseRule=UseColor_Specified, specifiedColor=<rgba 1 0.5 0 1>}
     uVRegion {bIsValid=false, max=<0 0>, min=<0 0>}
    }
   }
   slot {class=HorizontalBoxSlot, properties {horizontalAlignment=HAlign_Right, verticalAlignment=VAlign_Center, padding=<ltrb -9 0 10 0>}}
  }
  PlacementSettingsPanel {
   widgetClassPath=&EditorUtilityScrollBox
   bIsVariable=true
   parent=CanvasPanel_0
   widget=PlacementSettingsPanel
   slot {class=CanvasPanelSlot, properties {layoutData {alignment=<0 0>, anchors {maximum=<1 1>, minimum=<0 0>}, offsets=<ltrb 0 82 0 0>}}}
  }
  SelectVeneerMessage {widgetClassPath=&VerticalBox, parent=PlacementSettingsPanel, widget=SelectVeneerMessage, slot {class=ScrollBoxSlot}}
  TextBlock_56 {
   widgetClassPath=&TextBlock
   parent=SelectVeneerMessage
   widget=TextBlock_56
   properties {
    justification=Center
    text="Select or Create\r\nVeneer Placement Actor"
    font {
     bForceMonospaced=false
     bMaterialIsStencil=false
     fontMaterial=null
     fontObject=&Roboto
     letterSpacing=0
     monospacedWidth=1
     size=11
     skewAmount=0
     typefaceFontName=Regular
     outlineSettings {
      bApplyOutlineToDropShadows=false
      bMiteredCorners=false
      bSeparateFillAlpha=false
      outlineMaterial=null
      outlineSize=0
      outlineColor=<rgba 0 0 0 1>
     }
    }
   }
   slot {
    class=VerticalBoxSlot
    properties {horizontalAlignment=HAlign_Center, verticalAlignment=VAlign_Center, padding=<ltrb 0 100 0 0>, size {sizeRule=Fill, value=1}}
   }
  }
  VeneerProperties {widgetClassPath=&VerticalBox, parent=PlacementSettingsPanel, widget=VeneerProperties, slot {class=ScrollBoxSlot}}
  CustomPropertyView_46 {
   widgetClassPath=/Script/FortniteEditor.CustomPropertyView
   bIsVariable=true
   parent=VeneerProperties
   widget=CustomPropertyView_46
   properties {
    columns [{columnWidth=300, nameFillWidth=1, valueFillWidth=1}]
    propertiesToShow [{
      bForceShowParentProperty=false
      bForceShowProperty=false
      bShowChildren=true
      bShowContainerProperty=true
      bShowName=true
      bShowPropertiesOfSameName=false
      bShowPropertyButtons=true
      bShowValue=true
      columnNumber=0
      propertyName=VeneerType
      propertyNameOverride=""
     }]
   }
   slot {class=VerticalBoxSlot, properties {horizontalAlignment=HAlign_Center, verticalAlignment=VAlign_Center, padding=<ltrb 0 10 0 0>}}
  }
  HorizontalBox_3 {
   widgetClassPath=&HorizontalBox
   parent=VeneerProperties
   widget=HorizontalBox_3
   slot {class=VerticalBoxSlot, properties {horizontalAlignment=HAlign_Center, verticalAlignment=VAlign_Center, padding=<ltrb 0 10 0 10>}}
  }
  PinActors {
   widgetClassPath=&EditorUtilityButton
   parent=HorizontalBox_3
   widget=PinActors
   slot {class=HorizontalBoxSlot, properties {horizontalAlignment=HAlign_Center, verticalAlignment=VAlign_Center}}
  }
  CommonTextBlock_1 {
   widgetClassPath=&CommonTextBlock
   parent=PinActors
   widget=CommonTextBlock_1
   properties {
    text=Pin Selected Actors
    font {
     bForceMonospaced=false
     bMaterialIsStencil=false
     fontMaterial=null
     fontObject=&Roboto
     letterSpacing=0
     monospacedWidth=1
     size=9
     skewAmount=0
     typefaceFontName=Regular
     outlineSettings {
      bApplyOutlineToDropShadows=false
      bMiteredCorners=false
      bSeparateFillAlpha=false
      outlineMaterial=null
      outlineSize=0
      outlineColor=<rgba 0 0 0 1>
     }
    }
    shadowOffset=<0 0>
   }
   slot {class=ButtonSlot, properties {padding=<ltrb 0 6 0 6>}}
  }
  UnpinActors {
   widgetClassPath=&EditorUtilityButton
   parent=HorizontalBox_3
   widget=UnpinActors
   slot {class=HorizontalBoxSlot, properties {horizontalAlignment=HAlign_Center, verticalAlignment=VAlign_Center, padding=<ltrb 10 0 0 0>}}
  }
  CommonTextBlock_2 {
   widgetClassPath=&CommonTextBlock
   parent=UnpinActors
   widget=CommonTextBlock_2
   properties {
    text=Unpin Selected Actors
    font {
     bForceMonospaced=false
     bMaterialIsStencil=false
     fontMaterial=null
     fontObject=&Roboto
     letterSpacing=0
     monospacedWidth=1
     size=9
     skewAmount=0
     typefaceFontName=Regular
     outlineSettings {
      bApplyOutlineToDropShadows=false
      bMiteredCorners=false
      bSeparateFillAlpha=false
      outlineMaterial=null
      outlineSize=0
      outlineColor=<rgba 0 0 0 1>
     }
    }
    shadowOffset=<0 0>
   }
   slot {class=ButtonSlot, properties {padding=<ltrb 0 6 0 6>}}
  }
  Border_220 {
   widgetClassPath=&Border
   parent=VeneerProperties
   widget=Border_220
   properties {
    verticalAlignment=VAlign_Center
    background {
     bIsDynamicallyLoaded=false
     drawAs=Image
     imageType=NoImage
     mirroring=NoMirror
     resourceName=None
     resourceObject=null
     tiling=NoTile
     imageSize=<32 32>
     margin=<ltrb 0 0 0 0>
     outlineSettings {
      bUseBrushTransparency=false
      roundingType=HalfHeightRadius
      width=0
      color {colorUseRule=UseColor_Specified, specifiedColor=<rgba 0 0 0 0>}
      cornerRadii=<0 0 0 0>
     }
     tintColor {colorUseRule=UseColor_Specified, specifiedColor=<rgba 0.0075 0.0075 0.0075 0>}
     uVRegion {bIsValid=false, max=<0 0>, min=<0 0>}
    }
    padding=<ltrb 8 10 0 10>
   }
   slot {class=VerticalBoxSlot}
  }
  ObjectName_1 {
   widgetClassPath=&CommonTextBlock
   bIsVariable=true
   parent=Border_220
   widget=ObjectName_1
   properties {
    text=Parameters
    colorAndOpacity {colorUseRule=UseColor_Specified, specifiedColor=<rgba 0.9647 0.9647 0.9647 1>}
    font {
     bForceMonospaced=false
     bMaterialIsStencil=false
     fontMaterial=null
     fontObject=&Roboto
     letterSpacing=0
     monospacedWidth=1
     size=11
     skewAmount=0
     typefaceFontName=Regular
     outlineSettings {
      bApplyOutlineToDropShadows=false
      bMiteredCorners=false
      bSeparateFillAlpha=false
      outlineMaterial=null
      outlineSize=0
      outlineColor=<rgba 0 0 0 1>
     }
    }
    shadowOffset=<0 0>
   }
   slot {class=BorderSlot, properties {verticalAlignment=VAlign_Center, padding=<ltrb 8 10 0 10>}}
  }
  ParametersPanel {widgetClassPath=&VerticalBox, bIsVariable=true, parent=VeneerProperties, widget=ParametersPanel, slot {class=VerticalBoxSlot}}
  Border {
   widgetClassPath=&Border
   parent=VeneerProperties
   widget=Border
   properties {
    verticalAlignment=VAlign_Center
    background {
     bIsDynamicallyLoaded=false
     drawAs=Image
     imageType=NoImage
     mirroring=NoMirror
     resourceName=None
     resourceObject=null
     tiling=NoTile
     imageSize=<32 32>
     margin=<ltrb 0 0 0 0>
     outlineSettings {
      bUseBrushTransparency=false
      roundingType=HalfHeightRadius
      width=0
      color {colorUseRule=UseColor_Specified, specifiedColor=<rgba 0 0 0 0>}
      cornerRadii=<0 0 0 0>
     }
     tintColor {colorUseRule=UseColor_Specified, specifiedColor=<rgba 0.0075 0.0075 0.0075 0>}
     uVRegion {bIsValid=false, max=<0 0>, min=<0 0>}
    }
    padding=<ltrb 8 10 0 10>
   }
   slot {class=VerticalBoxSlot}
  }
  ObjectName_2 {
   widgetClassPath=&CommonTextBlock
   bIsVariable=true
   parent=Border
   widget=ObjectName_2
   properties {
    text=Coordinates
    colorAndOpacity {colorUseRule=UseColor_Specified, specifiedColor=<rgba 0.9647 0.9647 0.9647 1>}
    font {
     bForceMonospaced=false
     bMaterialIsStencil=false
     fontMaterial=null
     fontObject=&Roboto
     letterSpacing=0
     monospacedWidth=1
     size=11
     skewAmount=0
     typefaceFontName=Regular
     outlineSettings {
      bApplyOutlineToDropShadows=false
      bMiteredCorners=false
      bSeparateFillAlpha=false
      outlineMaterial=null
      outlineSize=0
      outlineColor=<rgba 0 0 0 1>
     }
    }
    shadowOffset=<0 0>
   }
   slot {class=BorderSlot, properties {verticalAlignment=VAlign_Center, padding=<ltrb 8 10 0 10>}}
  }
  CoordinatesPanel {widgetClassPath=&VerticalBox, bIsVariable=true, parent=VeneerProperties, widget=CoordinatesPanel, slot {class=VerticalBoxSlot}}
  ActorSettingsPanel {
   widgetClassPath=&EditorUtilityScrollBox
   bIsVariable=true
   parent=CanvasPanel_0
   widget=ActorSettingsPanel
   slot {class=CanvasPanelSlot, properties {layoutData {alignment=<0 0>, anchors {maximum=<1 1>, minimum=<0 0>}, offsets=<ltrb 0 82 0 0>}}}
  }
  DetailsView_70 {widgetClassPath=&DetailsView, parent=ActorSettingsPanel, widget=DetailsView_70, slot {class=ScrollBoxSlot}}
  MaterialsPanel {
   widgetClassPath=&EditorUtilityScrollBox
   bIsVariable=true
   parent=CanvasPanel_0
   widget=MaterialsPanel
   slot {class=CanvasPanelSlot, properties {layoutData {alignment=<0 0>, anchors {maximum=<1 1>, minimum=<0 0>}, offsets=<ltrb 0 82 0 82>}}}
  }
  MissingMaterialsPanel {widgetClassPath=&VerticalBox, bIsVariable=true, parent=MaterialsPanel, widget=MissingMaterialsPanel, slot {class=ScrollBoxSlot}}
  MissingMaterialsBatchPanel {
   widgetClassPath=&HorizontalBox
   bIsVariable=true
   parent=MaterialsPanel
   widget=MissingMaterialsBatchPanel
   slot {class=ScrollBoxSlot}
  }
  CreateAllMissingMaterialsBtn {
   widgetClassPath=&EditorUtilityButton
   parent=MissingMaterialsBatchPanel
   widget=CreateAllMissingMaterialsBtn
   slot {
    class=HorizontalBoxSlot
    properties {horizontalAlignment=HAlign_Center, verticalAlignment=VAlign_Center, padding=<ltrb 0 9 8 9>, size {sizeRule=Fill, value=1}}
   }
  }
  TextBlock_2 {
   widgetClassPath=&TextBlock
   parent=CreateAllMissingMaterialsBtn
   widget=TextBlock_2
   properties {
    text=Create All Missing Materials
    colorAndOpacity {colorUseRule=UseColor_Specified, specifiedColor=<rgba 0.5271 0.5271 0.5271 1>}
    font {
     bForceMonospaced=false
     bMaterialIsStencil=false
     fontMaterial=null
     fontObject=&Roboto
     letterSpacing=0
     monospacedWidth=1
     size=9.75
     skewAmount=0
     typefaceFontName=Regular
     outlineSettings {
      bApplyOutlineToDropShadows=false
      bMiteredCorners=false
      bSeparateFillAlpha=false
      outlineMaterial=null
      outlineSize=0
      outlineColor=<rgba 0 0 0 1>
     }
    }
    strikeBrush {
     bIsDynamicallyLoaded=false
     drawAs=Image
     imageType=NoImage
     mirroring=NoMirror
     resourceName=None
     resourceObject=null
     tiling=NoTile
     imageSize=<32 32>
     margin=<ltrb 0 0 0 0>
     outlineSettings {
      bUseBrushTransparency=false
      roundingType=HalfHeightRadius
      width=0
      color {colorUseRule=UseColor_Specified, specifiedColor=<rgba 0 0 0 0>}
      cornerRadii=<0 0 0 0>
     }
     tintColor {colorUseRule=UseColor_Specified, specifiedColor=<rgba 0.5271 0.5271 0.5271 1>}
     uVRegion {bIsValid=false, max=<0 0>, min=<0 0>}
    }
   }
   slot {class=ButtonSlot, properties {padding=<ltrb 0 9 0 9>}}
  }
  MaterialsInUsePanel {widgetClassPath=&VerticalBox, bIsVariable=true, parent=MaterialsPanel, widget=MaterialsInUsePanel, slot {class=ScrollBoxSlot}}
 ]
}
