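# Editor placement actor for the Veneer art tool, which swaps materials on pinned and optionally overlapping actors for veneer variants.
# It maintains a companion BP_VeneerContainer and adds or removes VeneerComponents on target actors through the Subobject Data subsystem.
# Unique original materials, including building texture-data overrides, map by VNR path naming to veneer assets, queuing missing ones.
# Components receive indices into the container's dynamic materials.
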EditorUtilityBlueprint /Game/ArtTools/Tech_Art/Veneer/EBP_VeneerPlacement
graphs [
 ActorsWithTextureData {nodes [
   n1 = ActorsWithTextureData!() -> (then, InActor:Actor Object Reference, AllowCustomMaterial:bool=~) {is=title K2Node_FunctionEntry}
   > n2 = Cast To BuildingSMActor!(execute, Object=@n1) -> (then, CastFailed, AsBuilding Static Mesh:Building Static Mesh Object Reference) {is=title K2Node_DynamicCast}
   > SetEditorProperty(Object=@n2, PropertyName=bAllowCustomMaterial, PropertyValue=@n1.AllowCustomMaterial)
   > SetEditorProperty(Object=@n2, PropertyName=bConsiderOverrideMaterialsOnAnimationCleanup, PropertyValue=@n1.AllowCustomMaterial)
  ]}
 AssignIndexToPrimitiveComponents {nodes [
   AssignIndexToPrimitiveComponents() {is=title K2Node_FunctionEntry}
   > n1 = Sequence!(execute) -> (then_0, then_1, then_2:Exec)
   > n2 = For Each Loop(Array:Array of Struct Veneer Original Materials Structures=Get OriginalMaterials() -> (OriginalMaterials:Array of Struct Veneer Original Materials Structures)) -> (Array Element:Struct Veneer Original Materials Structure)
   > Add(
    TargetArray:Array of Material Interface Object References=Get BaseMaterials() -> (BaseMaterials:Array of Material Interface Object References)
    NewItem:"Material Interface Object Reference (by ref)"=Break Struct Veneer Original Materials!(Struct_VeneerOriginalMaterials:"Struct Veneer Original Materials Structure (by ref)"=@n2) -> (
     OriginalMaterial_2_#1:Material Interface Object Reference
     TextureData_8_#2:Array of Building Texture Data Object References
    ) {is=title K2Node_BreakStruct}
   ) {is=title K2Node_CallArrayFunction}
   n3 = For Each Loop(
    Exec=@n1.then_1
    Array=Get VeneerBaseMaterials!(self:BP Veneer Container Object Reference=Get VeneerContainer() -> (VeneerContainer:BP Veneer Container Object Reference)) -> (VeneerBaseMaterials:Array of Struct Veneer Original Materials Structures)
   )
   > n4 = Cast To MaterialInstanceConstant(Object=Break Struct Veneer Original Materials(Struct_VeneerOriginalMaterials=@n3))
   > Add(
    TargetArray=Get VeneerMaterials() -> (VeneerMaterials:Array of Material Interface Object References)
    NewItem=Get Parent!(self:Material Instance Object Reference=@n4) -> (Parent:Material Interface Object Reference)
   )
   n5 = For Each Loop(
    Exec=@n1.then_2
    Array:Array of Veneer Component Object References=Get VeneerComponents() -> (VeneerComponents:Array of Veneer Component Object References)
   ) -> (Array Element:Veneer Component Object Reference)
   > n6 = For Each Loop(Array:Array of Struct Veneer Structures=Get PrimitiveComponentsToVeneer!(self:Veneer Component Object Reference=@n5) -> (PrimitiveComponentsToVeneer:Array of Struct Veneer Structures)) -> (Array Element:Struct Veneer Structure)
   > Clear(TargetArray:Array of Integers=Get LocalIndex() -> (LocalIndex:Array of Integers))
   n7 = Break Struct Veneer!(Struct_Veneer:"Struct Veneer Structure (by ref)"=@n6) -> (
    Primitive_2_#3:Primitive Component Object Reference
    bVeneer_8_#4:bool=~
    MaterialPointers_6_#5:Array of Integers
   ) {is=title K2Node_BreakStruct}
   > n8 = For Each Loop(Array:Array of Integers=@n7.MaterialPointers_6_#5) -> (Array Element:int)
   Set Array Elem!(
    execute=@n8.Completed
    TargetArray:Array of Struct Veneer Structures=Get PrimitiveComponentsToVeneer(self=@n5)
    Index:int=@n6.Array Index
    Item:"Struct Veneer Structure (by ref)"=Make Struct Veneer!(
     Primitive_2_#3:Primitive Component Object Reference=@n7
     bVeneer_8_#4:bool=@n7.bVeneer_8_#4
     MaterialPointers_6_#5:Array of Integers=Get LocalIndex()
    ) -> (Struct_Veneer:Struct Veneer Structure) {is=title K2Node_MakeStruct}
    bSizeToFit:bool=false
   ) -> (then)
   Branch(execute=@n8.LoopBody, Condition="Not Equal (Integer)"(A=@n8, B=-1))
   n9 = Find Item!(
    TargetArray:Array of Material Interface Object References=Get VeneerMaterials()
    ItemToFind:"Material Interface Object Reference (by ref)"="Get (a copy)"(Array:Array of Material Interface Object References=Get BaseMaterials(), Dimension 1=@n8) -> (Output:Material Interface Object Reference)
   ) -> (ReturnValue)
   > Add(TargetArray:Array of Integers=Get LocalIndex(), NewItem:"Integer (by ref)"=@n9)
   > Branch(Condition="Not Equal (Integer)"(A=@n9, B=-1))
   n10 = Get VeneerContainer()
   > SetMaterial(
    self=@n7
    ElementIndex=@n8.Array Index
    Material="Get (a copy)"(
     Array:Array of Material Instance Dynamic Object References=Get DynamicMaterials!(self:BP Veneer Container Object Reference=@n10) -> (DynamicMaterials:Array of Material Instance Dynamic Object References)
     Dimension 1=@n9
    ) -> (Output:Material Instance Dynamic Object Reference)
   )
   Assign Materials From Container(self:Veneer Component Object Reference=@n5)
   "Get (a ref)"(
    Array:Array of Booleans=Get Enable Veener On Component!(self:Veneer Component Object Reference=@n5) -> (Enable Veener On Component:Array of Booleans)
    Dimension 1=@n6.Array Index
   ) -> (Output:"Boolean (by ref)")
  ]}
 BindOnDelete {nodes [
   n1 = BindOnDelete!() -> (then, Unbind:bool=~) {is=title K2Node_FunctionEntry}
   > n2 = Branch(Condition=@n1)
   > Unbind all Events from On Delete Actors Begin!(execute:Exec, self:Editor Actor Subsystem Object Reference=Get EditorActorSubsystem()) -> (then:Exec) {is=title K2Node_ClearDelegate}
   Bind Event to On Delete Actors Begin(execute=@n2.else, self:Editor Actor Subsystem Object Reference=Get EditorActorSubsystem(), Delegate=Create Event()) {is=title K2Node_AddDelegate}
  ]}
 ConvertOriginalPathToVRN {nodes [
   n1 = Split(SourceString=Inputs!() -> (SourceString:String), InStr=/, SearchDir=FromEnd)
   n2 = Append!(A=@n1, B=/VNR/, C:String=@n1.RightS) -> (ReturnValue)
   n3 = "To Integer (Byte)"!(InByte:Byte=Get VeneerType() -> (VeneerType:Enum_VeneerType Enum=NewEnumerator0)) -> (ReturnValue:int=0)
   n4 = Make Array!("[0]":String=D, "[1]":String=DA, "[2]":String=C, "[3]":String=L) -> (Array:Array of Strings)
   Outputs!(ReturnValue:String=Append(
     A=Split(SourceString=@n2, InStr=., SearchDir=FromEnd)
     B=_VNR_
     C="Get (a ref)"(Array:Array of Strings=@n4, Dimension 1=@n3) -> (Output:"String (by ref)")
    ))
  ]}
 DeleteComponent {nodes [
   n1 = DeleteComponent!() -> (then, Component:Actor Component Object Reference) {is=title K2Node_FunctionEntry}
   n2 = GetOwner(self=@n1)
   > ActorsWithTextureData!(execute, self, InActor:Actor Object Reference=@n2, AllowCustomMaterial:bool=false) -> (then) {is=title K2Node_CallFunction}
   > n3 = Gather Subobject Data For Instance(self=Get SubobjectDataSubsystem(), Context=@n2)
   n4 = "Get (a ref)"(Array:Array of Subobject Data Handle Structures=@n3, Dimension 1=0) -> (Output:"Subobject Data Handle Structure (by ref)")
   n5 = Get SubobjectDataSubsystem()
   > Delete Subobject from Instance!(
    execute
    self:Subobject Data Subsystem Object Reference=@n5
    ContextHandle:"Subobject Data Handle Structure (by ref)"=@n4
    SubobjectToDelete:"Subobject Data Handle Structure (by ref)"=FindHandleForObject!(
     self:Subobject Data Subsystem Object Reference=@n5
     Context:"Subobject Data Handle Structure (by ref)"=@n4
     ObjectToFind:Object Reference=@n1
     BPContext:Blueprint Object Reference
    ) -> (ReturnValue:Subobject Data Handle Structure)
   ) -> (then, ReturnValue:int=0)
  ]}
 EventGraph {nodes [
   Open Veneer Widget() {is=title K2Node_CustomEvent}
   > SpawnAndRegisterTab(self=Get EditorUtilitySubsystem(), InBlueprint=&/EUW_Veneer)
   Update Veneer Placement() {is=title K2Node_CustomEvent}
   > UpdateVeneerContainer()
   > Manage Veneer Components()
  ]}
 FindUniqueOriginalMaterials {nodes [
   n1 = FindUniqueOriginalMaterials!() -> (then, Veneer Component:Veneer Component Object Reference) {is=title K2Node_FunctionEntry}
   > n2 = Set LocalVeneerComponent(LocalVeneerComponent:Veneer Component Object Reference=@n1) -> (Output_Get:Veneer Component Object Reference)
   > Set Local Actor(Local Actor:Actor Object Reference=GetOwner(self=@n2)) -> (Output_Get:Actor Object Reference)
   > n3 = For Each Loop(Array:Array of Struct Veneer Structures=Get PrimitiveComponentsToVeneer(self=@n2)) -> (Array Element:Struct Veneer Structure)
   > Set LoopItr(LoopItr:int=@n3.Array Index) -> (Output_Get:int=0)
   > Clear(TargetArray=Get MaterialIndex() -> (MaterialIndex:Array of Integers))
   > n4 = Set EnableVeneer(EnableVeneer:bool="Get (a ref)"(Array:Array of Booleans=Get Enable Veener On Component(self=@n2), Dimension 1=Get LoopItr() -> (LoopItr:int=0)) -> (Output:"Boolean (by ref)")) -> (Output_Get:bool)
   > n5 = Sequence!(execute) -> (then_0, then_1)
   n6 = Get LoopItr() -> (LoopItr=0)
   > Branch(Condition=AND Boolean(A=Get EnableVeneer() -> (EnableVeneer:bool), B="Equal (Integer)"(A=@n6)))
   > n7 = Cast To BuildingSMActor(Object=Get Local Actor() -> (Local Actor:Actor Object Reference))
   > n8 = GetBuildingTextureData!(execute, self:Building Static Mesh Object Reference=@n7) -> (then, OutTextureDataArray:Array of Building Texture Data Object References)
   n9 = Get OverrideMaterial!(self:Building Texture Data Object Reference="Get (a copy)"(Array:Array of Building Texture Data Object References=@n8) -> (Output:Building Texture Data Object Reference)) -> (OverrideMaterial:Material Interface Object Reference)
   n10 = Get StaticMesh(self:Building Static Mesh Object Reference=@n7)
   n11 = Make Struct Veneer Original Materials!(
    OriginalMaterial_2_#1:Material Interface Object Reference=Select(
     Option 0:Material Interface Object Reference=GetMaterial!(self:Static Mesh Object Reference=@n10, MaterialIndex:int=0) -> (ReturnValue:Material Interface Object Reference)
     Option 1:Material Interface Object Reference=@n9
     Index=IsValid(Object=@n9)
    ) -> (ReturnValue:Material Interface Object Reference)
    TextureData_8_#2:Array of Building Texture Data Object References=@n8
   ) -> (Struct_VeneerOriginalMaterials:Struct Veneer Original Materials Structure) {is=title K2Node_MakeStruct}
   > Add Unique(
    TargetArray:Array of Struct Veneer Original Materials Structures=Get OriginalMaterials()
    NewItem:"Struct Veneer Original Materials Structure (by ref)"=@n11
   )
   > Set HasTextureData(HasTextureData:bool=true) -> (Output_Get:bool)
   > Set BuildingActorStaticMeshComponent(BuildingActorStaticMeshComponent:Static Mesh Component Object Reference=Get StaticMeshComponent(self:Building Static Mesh Object Reference=@n7) -> (StaticMeshComponent:Base Building Static Mesh Component Object Reference)) -> (Output_Get:Static Mesh Component Object Reference)
   > Add(
    TargetArray=Get MaterialIndex()
    NewItem=Find Item(
     TargetArray:Array of Struct Veneer Original Materials Structures=Get OriginalMaterials()
     ItemToFind:"Struct Veneer Original Materials Structure (by ref)"=@n11
    )
   )
   Branch(execute=@n5.then_1, Condition=@n4)
   > n12 = Cast To StaticMeshComponent(Object=Break Struct Veneer(Struct_Veneer=@n3) -> (bVeneer_8_#4=~))
   > n13 = Set LocalStaticMeshComponent(LocalStaticMeshComponent:Static Mesh Component Object Reference=@n12) -> (Output_Get:Static Mesh Component Object Reference)
   > n14 = For Each Loop(Array:Array of Static Material Structures=Get StaticMaterials!(self:Static Mesh Object Reference=Get StaticMesh(self:Static Mesh Component Object Reference=@n13)) -> (StaticMaterials:Array of Static Material Structures)) -> (Array Element:Static Material Structure)
   n15 = Get PrimitiveComponentsToVeneer(self=Get LocalVeneerComponent() -> (LocalVeneerComponent:Veneer Component Object Reference))
   n16 = "Get (a ref)"!(Array:Array of Struct Veneer Structures=@n15, Dimension 1=Get LoopItr() -> (LoopItr=0)) -> (
    Output_Primitive_2_#3:Primitive Component Object Reference
    Output_bVeneer_8_#4:bool=~
    Output_MaterialPointers_6_#5:Array of Integers
   )
   Set Array Elem(
    execute=@n14.Completed
    TargetArray=@n15
    Index=Get LoopItr() -> (LoopItr=0)
    Item=Make Struct Veneer(
     Primitive_2_#3=@n16
     bVeneer_8_#4=@n16.Output_bVeneer_8_#4
     MaterialPointers_6_#5=Get MaterialIndex()
    )
    bSizeToFit=false
   )
   n17 = Get LocalStaticMeshComponent() -> (LocalStaticMeshComponent:Static Mesh Component Object Reference)
   n18 = Get BuildingActorStaticMeshComponent() -> (BuildingActorStaticMeshComponent:Static Mesh Component Object Reference)
   n19 = Branch(
    execute=@n14.LoopBody
    Condition=AND Boolean!(A="Equal (Object)"(A=@n17, B=@n18), B=Get HasTextureData() -> (HasTextureData:bool), C:bool="Equal (Integer)"(A=@n14.Array Index)) -> (ReturnValue)
   )
   n20 = Make Struct Veneer Original Materials(OriginalMaterial_2_#1=Break Static Material!(StaticMaterial:"Static Material Structure (by ref)"=@n14) -> (
     MaterialInterface:Material Interface Object Reference
     MaterialSlotName:Name
     UVChannelData:Mesh UVChannel Info Structure
     OverlayMaterialInterface:Material Interface Object Reference
    ) {is=title K2Node_BreakStruct})
   n21 = Get OriginalMaterials()
   Add Unique(execute=@n19.else, TargetArray=@n21, NewItem=@n20)
   > Add(TargetArray=Get MaterialIndex(), NewItem=Find Item(TargetArray=@n21, ItemToFind=@n20))
  ]}
 FindVeneerMaterialsFromOriginalMaterials {nodes [
   FindVeneerMaterialsFromOriginalMaterials() {is=title K2Node_FunctionEntry}
   > Clear!(
    execute
    TargetMap:Map of Material Interface Object References to Booleans=Get QueueMissingMaterials() -> (QueueMissingMaterials:Map of Material Interface Object References to Booleans)
   ) -> (then) {is=title K2Node_CallFunction}
   > Clear(TargetArray:Array of Struct Veneer Original Materials Structures=Get VeneerBaseMaterials(self=Get VeneerContainer())) {is=title K2Node_CallArrayFunction}
   > n1 = For Each Loop(Array:Array of Struct Veneer Original Materials Structures=Get OriginalMaterials()) -> (Array Element:Struct Veneer Original Materials Structure)
   n2 = Break Struct Veneer Original Materials(Struct_VeneerOriginalMaterials=@n1)
   > n3 = GetPathNameForLoadedAsset(LoadedAsset=@n2)
   n4 = Get EditorAssetSubsystem() -> (ReturnValue:Editor Asset Subsystem Object Reference) {is=title K2Node_GetEditorSubsystem}
   n5 = Convert Original Path to VRN!(SourceString:String=@n3) -> (ReturnValue:String) {is=title K2Node_MacroInstance}
   > n6 = DoesAssetExist!(execute, self:Editor Asset Subsystem Object Reference=@n4, AssetPath=@n5) -> (then, ReturnValue)
   > n7 = Branch(Condition=@n6)
   > n8 = LoadAsset!(execute, self:Editor Asset Subsystem Object Reference=@n4, AssetPath=@n5) -> (then, ReturnValue)
   > n9 = Cast To MaterialInterface!(execute, Object=@n8) -> (then, CastFailed, AsMaterial Interface:Material Interface Object Reference) {is=title K2Node_DynamicCast}
   > Add Unique(
    TargetArray=Get VeneerBaseMaterials(self=Get VeneerContainer())
    NewItem=Make Struct Veneer Original Materials(
     OriginalMaterial_2_#1=@n9
     TextureData_8_#2=@n2.TextureData_8_#2
    )
   )
   Add!(
    execute=@n7.else
    TargetMap:Map of Strings to Booleans=Get Paths() -> (Paths:Map of Strings to Booleans)
    Key:"String (by ref)"=@n5
    Value:"Boolean (by ref)"=@n6
   ) -> (then) {is=title K2Node_CallFunction}
   > n10 = Branch(Condition=@n6)
   Add(
    execute=@n10.else
    TargetMap:Map of Material Interface Object References to Booleans=Get QueueMissingMaterials()
    Key:"Material Interface Object Reference (by ref)"=@n2
    Value=MaterialSupportVNR!(Material:Material Interface Object Reference=@n2) -> (Support?:bool)
   )
  ]}
 GetMaterialParent {nodes [
   n1 = GetMaterialParent!() -> (then, In:Object Reference) {is=title K2Node_FunctionEntry}
   > n2 = Sequence()
   > n3 = Cast To Material!(execute, Object=@n1) -> (then, CastFailed, AsMaterial:Material Object Reference) {is=title K2Node_DynamicCast}
   > Set MaterialInterface(MaterialInterface:Material Interface Object Reference=@n3) -> (Output_Get:Material Interface Object Reference)
   n4 = Cast To MaterialInstanceConstant(execute=@n3.CastFailed, Object=@n1)
   > Set MaterialInterface(MaterialInterface=@n4)
   n5 = Cast To MaterialInstanceDynamic!(execute=@n4.CastFailed, Object=@n1) -> (then, CastFailed, AsMaterial Instance Dynamic:Material Instance Dynamic Object Reference) {is=title K2Node_DynamicCast}
   > Set MaterialInterface(MaterialInterface=Get Parent(self=@n5))
   Return Node!(
    execute=@n2.then_1
    Out:Material Interface Object Reference=Get MaterialInterface() -> (MaterialInterface:Material Interface Object Reference)
   )
  ]}
 GetOverlapping {nodes [
   GetOverlapping() {is=title K2Node_FunctionEntry}
   > n1 = For Each Loop(Array:Array of Actor Object References=GetOverlappingActors!(
     self:Primitive Component Object Reference=Get VeneerBox() -> (VeneerBox:Box Collision Object Reference)
     ClassFilter:Actor Class Reference=&Actor
    ) -> (OverlappingActors:Array of Actor Object References)) -> (Array Element:Actor Object Reference)
   n2 = Get PinnedActors() -> (PinnedActors:Array of Actor Object References)
   > n3 = Branch(Condition="Equal (Integer)"(A=Find Item(TargetArray:Array of Actor Object References=@n2, ItemToFind:"Actor Object Reference (by ref)"=@n1), B=-1))
   Is Valid(
    exec=@n3.else
    InputObject=GetComponentByClass!(self:Actor Object Reference=@n1, ComponentClass:Actor Component Class Reference=&StaticMeshComponent) -> (ReturnValue:Static Mesh Component Object Reference)
   )
   > Add Unique(
    TargetArray:Array of Actor Object References=Get OverlappingActors() -> (OverlappingActors:Array of Actor Object References)
    NewItem:"Actor Object Reference (by ref)"=@n1
   )
  ]}
 Manage Veneer Components {nodes [
   Manage Veneer Components() {is=title K2Node_FunctionEntry}
   > Clear(TargetArray=Get OriginalMaterials())
   > Clear(TargetArray:Array of Veneer Component Object References=Get VeneerComponents())
   > Clear(TargetArray:Array of Actor Object References=Get ActorsToVeneer() -> (ActorsToVeneer:Array of Actor Object References))
   > Append Array!(execute, TargetArray:Array of Actor Object References=Get ActorsToVeneer(), SourceArray:Array of Actor Object References=Get PinnedActors()) -> (then)
   > n1 = Branch(Condition=Get IncludeOverlappingActors() -> (IncludeOverlappingActors:bool))
   > n2 = Append Array(TargetArray=Get ActorsToVeneer(), SourceArray=Get OverlappingActors())
   n3 = For Each Loop(Exec=[@n2.then, @n1.else], Array=Get ActorsToVeneer())
   n4 = For Each Loop(Exec=@n3.Completed, Array=Get ActorsToVeneerBuffer() -> (ActorsToVeneerBuffer:Array of Actor Object References))
   Set ActorsToVeneerBuffer(execute=@n4.Completed, ActorsToVeneerBuffer:Array of Actor Object References=Get ActorsToVeneer()) -> (Output_Get:Array of Actor Object References)
   > Set Components!(
    execute
    Components:Array of Veneer Component Object References=Get VeneerComponents()
    self:BP Veneer Container Object Reference=Get VeneerContainer()
   ) -> (then, Output_Get:Array of Veneer Component Object References)
   n5 = Branch(
    execute=@n4.LoopBody
    Condition=Contains Item(TargetArray:Array of Actor Object References=Get ActorsToVeneer(), ItemToFind:"Actor Object Reference (by ref)"=@n4)
   )
   DeleteComponent!(
    execute=@n5.else
    self
    Component:Actor Component Object Reference=GetComponentByClass(self=@n4, ComponentClass=&/VeneerComponent.VeneerComponent_C) -> (ReturnValue:Veneer Component Object Reference)
   ) -> (then) {is=title K2Node_CallFunction}
   n6 = Set LocalActorToVeneer(execute=@n3.LoopBody, LocalActorToVeneer:Actor Object Reference=@n3) -> (Output_Get:Actor Object Reference)
   > Is Valid(InputObject=@n6)
   n7 = GetComponentByClass(
    self=Get LocalActorToVeneer() -> (LocalActorToVeneer:Actor Object Reference)
    ComponentClass=&VeneerComponent_C
   )
   > n8 = Is Valid(InputObject=@n7)
   n9 = Gather Subobject Data For Instance(execute=@n8.Is Not Valid, self=Get SubobjectDataSubsystem(), Context=Get LocalActorToVeneer())
   > AddNewSubobject!(
    execute
    self:Subobject Data Subsystem Object Reference=Get SubobjectDataSubsystem()
    Params:"Add New Subobject Params Structure (by ref)"=Make Add New Subobject Params!(
     ParentHandle:Subobject Data Handle Structure="Get (a ref)"!(Array:Array of Subobject Data Handle Structures=@n9, Dimension 1=0) -> (Output:"Subobject Data Handle Structure (by ref)")
     NewClass:Object Class Reference=&VeneerComponent_C
     BlueprintContext:Blueprint Object Reference
     bSkipMarkBlueprintModified:bool=False
     bConformTransformToParent:bool=True
    ) -> (AddNewSubobjectParams:Add New Subobject Params Structure) {is=title K2Node_MakeStruct}
   ) -> (then, FailReason:Text, ReturnValue:Subobject Data Handle Structure)
   n10 = GetComponentByClass(self=Get LocalActorToVeneer(), ComponentClass=&VeneerComponent_C)
   > Is Valid(InputObject=@n10)
   n11 = Get VeneerContainer()
   > Is Valid(InputObject=@n11)
   > Set VeneerContainer!(execute, VeneerContainer:BP Veneer Container Object Reference=@n11, self:Veneer Component Object Reference=@n10) -> (then, Output_Get:BP Veneer Container Object Reference)
   > Gather Primitive Components!(execute, self:Veneer Component Object Reference=@n10, Initilize:bool=true) -> (then)
   > ActorsWithTextureData(InActor=Get LocalActorToVeneer(), AllowCustomMaterial=true)
   Branch(
    execute=@n8.Is Valid
    Condition="Equal (Object)"(A=Get VeneerContainer(), B=Get VeneerContainer!(self:Veneer Component Object Reference=@n7) -> (VeneerContainer))
   )
   > Add Unique(TargetArray:Array of Veneer Component Object References=Get VeneerComponents(), NewItem:"Veneer Component Object Reference (by ref)"=@n7)
   > Gather Primitive Components(self=@n7, Initilize=false)
   > FindUniqueOriginalMaterials!(execute, self, Veneer Component:Veneer Component Object Reference=@n7) -> (then) {is=title K2Node_CallFunction}
   > FindVeneerMaterialsFromOriginalMaterials() {is=title K2Node_CallFunction}
   > AssignIndexToPrimitiveComponents() {is=title K2Node_CallFunction}
   DeleteComponent()
   FindVeneerMaterialsFromOriginalMaterials()
  ]}
 MatchMaterialIndex {nodes [
   n1 = MatchMaterialIndex!() -> (then, Veneer Component:Veneer Component Object Reference) {is=title K2Node_FunctionEntry}
   > n2 = Set Local Component(Local Component:Veneer Component Object Reference=@n1) -> (Output_Get:Veneer Component Object Reference)
   > n3 = Sequence()
   > n4 = Cast To BuildingSMActor(Object=GetOwner(self=Get Local Component() -> (Local Component:Veneer Component Object Reference)))
   > n5 = GetBaseMaterial!(
    execute
    self:Material Interface Object Reference=GetMaterial!(self:Primitive Component Object Reference=Get StaticMeshComponent(self=@n4), ElementIndex:int=0) -> (ReturnValue)
   ) -> (then, ReturnValue:Material Object Reference)
   > n6 = GetBuildingTextureData(self=@n4)
   n7 = For Each Loop(Exec=@n3.then_1, Array:Array of Struct Veneer Structures=Get PrimitiveComponentsToVeneer(self=@n2)) -> (Array Element:Struct Veneer Structure)
   > Branch(Condition=Break Struct Veneer(Struct_Veneer=@n7) -> (bVeneer_8_#4=~).bVeneer_8_#4)
   Find Item(
    TargetArray:Array of Struct Veneer Original Materials Structures=Get OriginalMaterials()
    ItemToFind:"Struct Veneer Original Materials Structure (by ref)"=Make Struct Veneer Original Materials(OriginalMaterial_2_#1=@n5, TextureData_8_#2=@n6)
   )
  ]}
 MaterialSupportVNR {nodes [
   n1 = MaterialSupportVNR!() -> (then, Material:Material Interface Object Reference) {is=title K2Node_FunctionEntry}
   > n2 = GetStaticSwitchParameterSource!(execute, Material:Material Interface Object Reference=@n1, ParameterName:Name=Enable DirectFX) -> (then, ParameterSource:Soft Object Path Structure, ReturnValue:bool)
   > Return Node!(execute, Support?:bool=@n2.ReturnValue)
  ]}
 OnDelete {nodes [
   OnDelete() {is=title K2Node_FunctionEntry}
   > n1 = For Each Loop(Array:Array of Veneer Component Object References=Get VeneerComponents()) -> (Array Element:Veneer Component Object Reference)
   Destroy Actor(execute=@n1.Completed, self=Get VeneerContainer!() -> (VeneerContainer))
   > Destroy Actor()
   DeleteComponent(execute=@n1.LoopBody, Component=@n1)
  ]}
 UpdateVeneerContainer {nodes [
   n1 = UpdateVeneerContainer() {is=title K2Node_FunctionEntry}
   n2 = Is Valid(exec=[@n10.then, @n1.then, @n9.else], InputObject=Get VeneerContainer())
   n3 = SpawnActorFromClass!(
    execute=[@n2.Is Not Valid, @n7.else]
    ActorClass:Actor Class Reference=&/BP_VeneerContainer.BP_VeneerContainer_C
    Location:Vector=Get Actor Location()
    Rotation:Rotator
    bTransient:bool
   ) -> (then, ReturnValue:BP Veneer Container Object Reference)
   > n4 = Set VeneerContainer!(execute, VeneerContainer=@n3) -> (then, Output_Get)
   > SetActorLabel(self=@n4, NewActorLabel=Append!(A=GetActorLabel(bCreateIfNone=true), B=_Container) -> (ReturnValue), bMarkDirty=true)
   Set Actor Transform!(
    execute=@n2.Is Valid
    self:Actor Object Reference=Get VeneerContainer()
    NewTransform:"Transform (by ref)"=Get Actor Transform()
    bSweep:bool=true
    bTeleport:bool=true
   ) -> (then, SweepHitResult:Hit Result Structure, ReturnValue:bool) {is=title K2Node_CallFunction}
   > SetEditorProperty(Object=Get VeneerContainer(), PropertyName=bHidden, PropertyValue=NOT Boolean(A=Get DummyBool() -> (DummyBool:bool)))
   n5 = Get VeneerContainer()
   n6 = GetActorLabel(bCreateIfNone=true)
   n7 = Branch(Condition="Equal, Case Insensitive (String)"(A:String=GetActorLabel(self=@n5, bCreateIfNone=true), B:String=Append(A=@n6, B=_Container)) -> (ReturnValue:bool=~) {is=title K2Node_PromotableOperator})
   n8 = Split(SourceString=GetActorLabel(bCreateIfNone=true), InStr=EBP_, SearchDir=FromStart)
   n9 = Branch(Condition=@n8.ReturnValue)
   > n10 = SetActorLabel(NewActorLabel=@n8.RightS, bMarkDirty=true)
  ]}
 UserConstructionScript {nodes [
   Construct Dynamic Materials!(execute, self:BP Veneer Container Object Reference=Get VeneerContainer(), TransformOnly:bool=false) -> (then)
   > UpdateVeneerContainer() {is=title K2Node_CallFunction}
   n1 = For Each Loop(Array=Get VeneerComponents())
   > Assign Materials From Container(self=@n1)
   GetOverlapping() {is=title K2Node_CallFunction}
   Construction Script()
  ]}
]
variables [
 ActorsToVeneer {type="array<Actor>"}
 ActorsToVeneerBuffer {type="array<Actor>"}
 IncludeOverlappingActors {type=boolean}
 OriginalMaterials {type="array<Struct_VeneerOriginalMaterials>", default []}
 OverlappingActors {type="array<Actor>"}
 Paths {type="map<boolean>", default {}}
 PinnedActors {type="array<Actor>"}
 QueueMissingMaterials {}
 VeneerComponents {type="array<VeneerComponent_C>", default []}
 VeneerContainer {type=BP_VeneerContainer_C, default=null}
 VeneerType {type=Enum_VeneerType, default=Decal}
]
